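# EUW_ExhaustPopButton is an editor utility widget in the engine audio folder: a single button labelled with an exhaust pop name.
# On Construct, it converts the ExhaustPopNames string to text and writes it to the button's text block.
# The ExhaustPops SoundWave array, a boolean and a calendar-offset struct are declared but unused in the graph.
# No click handler exists, and PreConstruct and Tick are empty.
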EditorUtilityWidgetBlueprint /VehicleAudio/EngineAudio/EUW_ExhaustPopButton
graphs [EventGraph {nodes [
   Event Construct()
   > Set Text(
    Text="To Text (String)"(InString=Get ExhaustPopNames() -> (ExhaustPopNames:String))
    self=Get TextBlock_173() -> (TextBlock_173:Text Object Reference)
   )
   Event PreConstruct()
   Event Tick()
  ]}]
variables [
 ExhaustPopNames {type=string, default=""}
 ExhaustPops {type="array<SoundWave>", default []}
 NewVar {type=boolean}
 NewVar_0 {
  type=CalendarOffsetEvent_Struct
  default {
   calendar Event Name=""
   offset Start Position=0.5
   "time Offset (Days)"=0
   "time Offset (Hours)"=0
   "time Offset (Minutes)"=0
   "time Offset (Seconds)"=0
   trigger if in the Past=false
  }
 }
]
widgets {
 info {rootWidgetClass=&EditorUtilityButton, widgetCount=2}
 widgets [
  EditorUtilityButton_158 {widgetClassPath=&EditorUtilityButton, widget=EditorUtilityButton_158, slot=~}
  TextBlock_173 {
   widgetClassPath=&TextBlock
   bIsVariable=true
   parent=EditorUtilityButton_158
   widget=TextBlock_173
   properties {
    text=ExhaustPopName
    font {
     bForceMonospaced=false
     bMaterialIsStencil=false
     fontMaterial=null
     fontObject=&Roboto
     letterSpacing=0
     monospacedWidth=1
     size=11.25
     skewAmount=0
     typefaceFontName=Bold
     outlineSettings {
      bApplyOutlineToDropShadows=false
      bMiteredCorners=false
      bSeparateFillAlpha=false
      outlineMaterial=null
      outlineSize=0
      outlineColor=<rgba 0 0 0 1>
     }
    }
   }
   slot {class=ButtonSlot}
  }
 ]
}
