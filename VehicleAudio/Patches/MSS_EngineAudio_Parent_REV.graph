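# Parent MetaSound source that produces looping stereo vehicle engine audio from RPM, throttle, gear and speed.
# A REV player feeds throttle-filtered engine (plus blowoff), waveshaped exhaust (plus shift/blowoff pops) and stereo-widened interior channels.
# These join a transmission layer in a game-specific mixer, then pass through perspective volume and EQ, azimuth panning and final four-band EQ.
# Vehicle-interface and directly named inputs are merged, so either can drive it.
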MetaSoundSource /VehicleAudio/Patches/MSS_EngineAudio_Parent_REV
inputs [
 Acceleration {type=Float}
 Blowoff.Sound {type=WaveAsset:Array, access=Constructor}
 Blowoff.Volume {type=Float, access=Constructor}
 BonusSpeed {type=Float}
 Character.Affiliation {type=Enum:FortAudioAffiliation}
 Character.Health {type=Float}
 Character.UniqueID {type=Int32}
 Downshift {type=Trigger}
 DriftAngle {type=Float}
 Engine.Throttle.EQ.FilterQ {type=Float, access=Constructor}
 Engine.Throttle.EQ.FilterType {type=Enum:BiquadFilterType, access=Constructor}
 Engine.Throttle.EQ.OffFreq {type=Float, access=Constructor}
 Engine.Throttle.EQ.OffGain {type=Float, access=Constructor}
 Engine.Throttle.EQ.OnFreq {type=Float, access=Constructor}
 Engine.Throttle.EQ.OnGain {type=Float, access=Constructor}
 Engine.Throttle.InterpTime {type=Time, access=Constructor}
 Engine.Throttle.Volume.Off {type=Float, access=Constructor}
 Engine.Throttle.Volume.On {type=Float, access=Constructor}
 EngineBlowoff {type=Trigger}
 Exhaust.Pops.Blowoff.Volume {type=Float, access=Constructor}
 Exhaust.Pops.Long {type=WaveAsset:Array, access=Constructor}
 Exhaust.Pops.Pitch {type=Float, access=Constructor}
 Exhaust.Pops.Shift.Volume {type=Float, access=Constructor}
 Exhaust.Pops.Short {type=WaveAsset:Array, access=Constructor}
 Exhaust.Throttle.EQ.FilterQ {type=Float, access=Constructor}
 Exhaust.Throttle.EQ.FilterType {type=Enum:BiquadFilterType, access=Constructor}
 Exhaust.Throttle.EQ.OffFreq {type=Float, access=Constructor}
 Exhaust.Throttle.EQ.OffGain {type=Float, access=Constructor}
 Exhaust.Throttle.EQ.OnFreq {type=Float, access=Constructor}
 Exhaust.Throttle.EQ.OnGain {type=Float, access=Constructor}
 Exhaust.Throttle.InterpTime {type=Time, access=Constructor}
 Exhaust.Throttle.Volume.Off {type=Float, access=Constructor}
 Exhaust.Throttle.Volume.On {type=Float, access=Constructor}
 Exhaust.Throttle.Waveshaper.Amount {type=Float, access=Constructor}
 Exhaust.Throttle.Waveshaper.AmountOffScale {type=Float, access=Constructor}
 Exhaust.Throttle.Waveshaper.AmountOnScale {type=Float, access=Constructor}
 Exhaust.Throttle.Waveshaper.Bias {type=Float, access=Constructor}
 Exhaust.Throttle.Waveshaper.Bypass {type=Bool, access=Constructor}
 Exhaust.Throttle.Waveshaper.Crossover {type=Float, access=Constructor}
 Exhaust.Throttle.Waveshaper.Mix {type=Float, access=Constructor}
 Exhaust.Throttle.Waveshaper.OutputGain {type=Float, access=Constructor}
 Exhaust.Throttle.Waveshaper.Type {type=Enum:WaveShaperType, access=Constructor}
 Fort.Vehicle.Acceleration {type=Float}
 Fort.Vehicle.BlowoffStrength {type=Float}
 Fort.Vehicle.BonusSpeed {type=Float}
 Fort.Vehicle.EnableShifting {type=Bool}
 Fort.Vehicle.Gear {type=Int32}
 Fort.Vehicle.IsGrounded {type=Bool}
 Fort.Vehicle.Load {type=Float}
 Fort.Vehicle.NormalSpeed {type=Float}
 Fort.Vehicle.OnBlowoff {type=Trigger}
 Fort.Vehicle.OnDownshift {type=Trigger}
 Fort.Vehicle.OnUpshift {type=Trigger}
 Fort.Vehicle.Pitch {type=Float}
 Fort.Vehicle.RPM {type=Float}
 Fort.Vehicle.Throttle {type=Float}
 GearChangeTime {type=Time, access=Constructor}
 Interior.EQ.CutoffFrequency {type=Float, access=Constructor}
 Interior.EQ.FilterType {type=Enum:BiquadFilterType, access=Constructor}
 Interior.StereoEffect.Crossover {type=Float, access=Constructor}
 Interior.StereoEffect.Mix {type=Float, access=Constructor}
 IsLocal {type=Bool}
 Mix.Azimuth.Engine.FilterType {type=Enum:BiquadFilterType, access=Constructor}
 Mix.Azimuth.Engine.Freq.Max {type=Float, access=Constructor}
 Mix.Azimuth.Engine.Freq.Min {type=Float, access=Constructor}
 Mix.Azimuth.Engine.MinVolume {type=Float, access=Constructor}
 Mix.Azimuth.Exhaust.FilterType {type=Enum:BiquadFilterType, access=Constructor}
 Mix.Azimuth.Exhaust.Freq.Max {type=Float, access=Constructor}
 Mix.Azimuth.Exhaust.Freq.Min {type=Float, access=Constructor}
 Mix.Azimuth.Exhaust.MinVolume {type=Float, access=Constructor}
 Mix.EQ1.FilterType {type=Enum:BiquadFilterType, access=Constructor}
 Mix.EQ1.Freq {type=Float, access=Constructor}
 Mix.EQ1.Gain {type=Float, access=Constructor}
 Mix.EQ1.Q {type=Float, access=Constructor}
 Mix.EQ2.FilterType {type=Enum:BiquadFilterType, access=Constructor}
 Mix.EQ2.Freq {type=Float, access=Constructor}
 Mix.EQ2.Gain {type=Float, access=Constructor}
 Mix.EQ2.Q {type=Float, access=Constructor}
 Mix.EQ3.FilterType {type=Enum:BiquadFilterType, access=Constructor}
 Mix.EQ3.Freq {type=Float, access=Constructor}
 Mix.EQ3.Gain {type=Float, access=Constructor}
 Mix.EQ3.Q {type=Float, access=Constructor}
 Mix.EQ4.FilterType {type=Enum:BiquadFilterType, access=Constructor}
 Mix.EQ4.Freq {type=Float, access=Constructor}
 Mix.EQ4.Gain {type=Float, access=Constructor}
 Mix.EQ4.Q {type=Float, access=Constructor}
 Mix.MixPerspective {type=Int32}
 Mix.Mute.Engine {type=Bool, access=Constructor}
 Mix.Mute.Exhaust {type=Bool, access=Constructor}
 Mix.Mute.Interior {type=Bool, access=Constructor}
 Mix.Perspective.Exterior.Engine.EQ {type=Float:Array, access=Constructor}
 Mix.Perspective.Exterior.Exhaust.EQ {type=Float:Array, access=Constructor}
 Mix.Perspective.Exterior.Volume {type=Float:Array, access=Constructor}
 Mix.Perspective.Interior.Engine.EQ {type=Float:Array, access=Constructor}
 Mix.Perspective.Interior.Exhaust.EQ {type=Float:Array, access=Constructor}
 Mix.Perspective.Interior.Volume {type=Float:Array, access=Constructor}
 Multichannel {type=REV Asset, access=Constructor}
 NormalizedSpeed {type=Float}
 Pitch {type=Float}
 Quality {type=Bool}
 REV.EnableShifting {type=Bool}
 REV.Gear {type=Int32}
 REV.RPM {type=Float}
 REV.Throttle {type=Float}
 REV.Velocity {type=Float}
 REVVolume {type=Float}
 Settings {type=REV Settings, access=Constructor}
 Transmission.Enabled {type=Bool}
 Transmission.Gear.Volume {type=Float:Array, access=Constructor}
 Transmission.MaxFreq {type=Float, access=Constructor}
 Transmission.MinFreq {type=Float, access=Constructor}
 Transmission.Throttle.Volume.Off {type=Float, access=Constructor}
 Transmission.Throttle.Volume.On {type=Float, access=Constructor}
 Transmission.Volume.Overall {type=Float, access=Constructor}
 UE.Source.OnPlay {type=Trigger}
 UE.Source.Orientation.Azimuth {type=Float}
 UE.Source.Orientation.Elevation {type=Float}
 Upshift {type=Trigger}
]
interfaces [Fort.Character:1.0, Fort.Vehicle:1.0, REV:1.0, UE.OutputFormat.Stereo:1.0, UE.Source.Orientation:1.0, UE.Source:1.1]
outputs [
 "UE.OutputFormat.Stereo.Audio:0" {type=Audio}
 "UE.OutputFormat.Stereo.Audio:1" {type=Audio}
]
pages [Default {
  input_defaults {
   Acceleration=0
   Blowoff.Volume=0
   BonusSpeed=0
   Character.Affiliation=0
   Character.Health=0
   Character.UniqueID=0
   Downshift=null
   DriftAngle=0
   Engine.Throttle.EQ.FilterQ=1
   Engine.Throttle.EQ.FilterType=1
   Engine.Throttle.EQ.OffFreq=250
   Engine.Throttle.EQ.OffGain=0
   Engine.Throttle.EQ.OnFreq=20
   Engine.Throttle.EQ.OnGain=0
   Engine.Throttle.InterpTime=0.1
   Engine.Throttle.Volume.Off=0
   Engine.Throttle.Volume.On=0
   EngineBlowoff=null
   Exhaust.Pops.Blowoff.Volume=0
   Exhaust.Pops.Pitch=0
   Exhaust.Pops.Shift.Volume=0
   Exhaust.Throttle.EQ.FilterQ=1
   Exhaust.Throttle.EQ.FilterType=4
   Exhaust.Throttle.EQ.OffFreq=5500
   Exhaust.Throttle.EQ.OffGain=0
   Exhaust.Throttle.EQ.OnFreq=5500
   Exhaust.Throttle.EQ.OnGain=0
   Exhaust.Throttle.InterpTime=0.1
   Exhaust.Throttle.Volume.Off=0
   Exhaust.Throttle.Volume.On=0
   Exhaust.Throttle.Waveshaper.Amount=0
   Exhaust.Throttle.Waveshaper.AmountOffScale=0
   Exhaust.Throttle.Waveshaper.AmountOnScale=1
   Exhaust.Throttle.Waveshaper.Bias=0
   Exhaust.Throttle.Waveshaper.Bypass=null
   Exhaust.Throttle.Waveshaper.Crossover=272
   Exhaust.Throttle.Waveshaper.Mix=1
   Exhaust.Throttle.Waveshaper.OutputGain=0.5
   Exhaust.Throttle.Waveshaper.Type=4
   Fort.Vehicle.Acceleration=0
   Fort.Vehicle.BlowoffStrength=0
   Fort.Vehicle.BonusSpeed=0
   Fort.Vehicle.EnableShifting=null
   Fort.Vehicle.Gear=0
   Fort.Vehicle.IsGrounded="<MetaSoundBuilderResult.SUCCEEDED: 0>"
   Fort.Vehicle.Load=0
   Fort.Vehicle.NormalSpeed=0
   Fort.Vehicle.OnBlowoff=null
   Fort.Vehicle.OnDownshift=null
   Fort.Vehicle.OnUpshift=null
   Fort.Vehicle.Pitch=0
   Fort.Vehicle.RPM=0
   Fort.Vehicle.Throttle=0
   GearChangeTime=0.1
   Interior.EQ.CutoffFrequency=5500
   Interior.EQ.FilterType=0
   Interior.StereoEffect.Crossover=494
   Interior.StereoEffect.Mix=0.2
   IsLocal=null
   Mix.Azimuth.Engine.FilterType=0
   Mix.Azimuth.Engine.Freq.Max=20000
   Mix.Azimuth.Engine.Freq.Min=5500
   Mix.Azimuth.Engine.MinVolume=-6
   Mix.Azimuth.Exhaust.FilterType=1
   Mix.Azimuth.Exhaust.Freq.Max=20
   Mix.Azimuth.Exhaust.Freq.Min=500
   Mix.Azimuth.Exhaust.MinVolume=-6
   Mix.EQ1.FilterType=4
   Mix.EQ1.Freq=100
   Mix.EQ1.Gain=0
   Mix.EQ1.Q=1
   Mix.EQ2.FilterType=4
   Mix.EQ2.Freq=300
   Mix.EQ2.Gain=-2
   Mix.EQ2.Q=1
   Mix.EQ3.FilterType=4
   Mix.EQ3.Freq=5000
   Mix.EQ3.Gain=0
   Mix.EQ3.Q=1
   Mix.EQ4.FilterType=4
   Mix.EQ4.Freq=7657.6494
   Mix.EQ4.Gain=0
   Mix.EQ4.Q=1
   Mix.MixPerspective=0
   Mix.Mute.Engine=null
   Mix.Mute.Exhaust=null
   Mix.Mute.Interior=null
   Multichannel=null
   NormalizedSpeed=0
   Pitch=0
   Quality=null
   REV.EnableShifting=null
   REV.Gear=6
   REV.RPM=0
   REV.Throttle=0
   REV.Velocity=0
   REVVolume=0
   Settings=/VehicleAudio/EngineAudio/Ferrari458/REVSetting_Default
   Transmission.Enabled=null
   Transmission.MaxFreq=250
   Transmission.MinFreq=60
   Transmission.Throttle.Volume.Off=0
   Transmission.Throttle.Volume.On=0
   Transmission.Volume.Overall=0
   UE.Source.OnPlay=null
   UE.Source.Orientation.Azimuth=0
   UE.Source.Orientation.Elevation=0
   Upshift=null
   Blowoff.Sound []
   Exhaust.Pops.Long []
   Exhaust.Pops.Short []
   Mix.Perspective.Exterior.Engine.EQ [20,1,0,1]
   Mix.Perspective.Exterior.Exhaust.EQ [20,1,0,1]
   Mix.Perspective.Exterior.Volume [0,0,-96]
   Mix.Perspective.Interior.Engine.EQ [20,1,0,1]
   Mix.Perspective.Interior.Exhaust.EQ [20,1,0,1]
   Mix.Perspective.Interior.Volume [-96,-96,0]
   Transmission.Gear.Volume [0,0,0,0,0,0,0]
  }
  nodes [
   TriggerAnyTriggerAny2_2 = "TriggerAny.Trigger Any (2)"(In 0=^EngineBlowoff, In 1=^Fort.Vehicle.OnBlowoff)
   n1 = InitVariable.Float()
   n2 = MapRange.MapRange:Float(In=^Fort.Vehicle.BonusSpeed) {also {defaults {Out Range B=0.75}}}
   n3 = VariableMutator.Float(
    Value=Clamp.Clamp:Float(In=Max.Max:Float(B=^Fort.Vehicle.RPM, A=^REV.RPM).Value) {also {defaults {Max=1}}}.Value
    Variable=InitVariable.Float().Variable
   )
   n4 = UE.Add:Float(AdditionalOperands=^Fort.Vehicle.Throttle, PrimaryOperand=^REV.Throttle)
   n5 = VariableDeferredAccessor.Float(Variable=@n1.Variable)
   n6 = VariableDeferredAccessor.Float(Variable=@n5.Variable)
   n7 = VariableAccessor.Float(Variable=VariableMutator.Float(Value=Clamp.Clamp:Float(In=@n4.Out) {also {defaults {Max=1, Min=0}}}.Value, Variable=@n6.Variable).Variable)
   n8 = VariableAccessor.Int32(Variable=VariableMutator.Int32(Variable=InitVariable.Int32().Variable, Value=Max.Max:Int32(B=^Fort.Vehicle.Gear, A=^REV.Gear).Value).Variable)
   REVRevPlayer_0 = REV.Rev Player:8 Channel(
    REV Asset=^Multichannel
    Enable Shifting=^REV.EnableShifting
    REV Settings=^Settings
    Play=^UE.Source.OnPlay
    Pitch=UE.Add:Float(AdditionalOperands=^Fort.Vehicle.Pitch, PrimaryOperand=UE.InterpTo:Audio(Target=@n2.Out Value) {also {defaults {Interp Time=0.15}}}.Value).Out
    RPM=&/DM_EngineAudio_RoadNoiseParamModifier(Input=VariableAccessor.Float(Variable=@n3.Variable).Value).Out
    Throttle=&DM_EngineAudio_RoadNoiseParamModifier(Input=@n7.Value).Out
    Gear=@n8.Value
   )
   n9 = VariableAccessor.Float(Variable=@n7.Variable)
   DecibelstoLinearGain_0 = UE.Decibels to Linear Gain:Float(Decibels=^REVVolume)
   AudioMixerAudioMixerMono2_0 = "AudioMixer.Audio Mixer (Mono, 2)"(
    In 0=&/DM_EngineAudio_Blowoff(
     BlowoffSound=^Blowoff.Sound
     OutputVolume=^Blowoff.Volume
     Character.Affiliation=^Character.Affiliation
     IsLocal=^IsLocal
     UE.Source.OnPlay=^UE.Source.OnPlay
     EngineBlowoff=@TriggerAnyTriggerAny2_2.Out
     REV.Throttle=VariableDeferredAccessor.Float(Variable=@n1.Variable).Value
    )."UE.OutputFormat.Mono.Audio:0"
    In 1=&/DM_EngineAudio_ThrottleVolumeFilter(
     Filter.FilterQ=^Engine.Throttle.EQ.FilterQ
     Filter.FilterType=^Engine.Throttle.EQ.FilterType
     Filter.OffThrottleFreq=^Engine.Throttle.EQ.OffFreq
     Filter.OnThrottleGain=^Engine.Throttle.EQ.OffGain
     Filter.OnThrottleFreq=^Engine.Throttle.EQ.OnFreq
     Filter.OffThrottleGain=^Engine.Throttle.EQ.OnGain
     Interp Time=^Engine.Throttle.InterpTime
     Volume.OffThrottle=^Engine.Throttle.Volume.Off
     Volume.OnThrottle=^Engine.Throttle.Volume.On
     AudioIn=@REVRevPlayer_0.Channel 1
     REV.Throttle=VariableAccessor.Float(Variable=@n9.Variable).Value
    ).AudioOut
    Gain 1=@DecibelstoLinearGain_0.Linear Gain
   )
   TriggerAnyTriggerAny2_0 = "TriggerAny.Trigger Any (2)"(In 1=^EngineBlowoff, In 0=^Fort.Vehicle.OnBlowoff)
   TriggerAnyTriggerAny2_1 = "TriggerAny.Trigger Any (2)"(In 1=^Downshift, In 0=^Fort.Vehicle.OnDownshift)
   TriggerAnyTriggerAny2_3 = "TriggerAny.Trigger Any (2)"(In 1=^Fort.Vehicle.OnUpshift, In 0=^Upshift)
   DecibelstoLinearGain_1 = UE.Decibels to Linear Gain:Float(Decibels=^REVVolume)
   AudioMixerAudioMixerMono2_1 = "AudioMixer.Audio Mixer (Mono, 2)"(
    In 0=&/DM_EngineAudio_ExhaustPops(
     Character.Affiliation=^Character.Affiliation
     BlowoffVolume=^Exhaust.Pops.Blowoff.Volume
     ExhaustPops_Long=^Exhaust.Pops.Long
     Pitch=^Exhaust.Pops.Pitch
     ShiftVolume=^Exhaust.Pops.Shift.Volume
     ExhaustPops_Short=^Exhaust.Pops.Short
     GearChangeTime=^GearChangeTime
     IsLocal=^IsLocal
     Blowoff=@TriggerAnyTriggerAny2_0.Out
     Downshift=@TriggerAnyTriggerAny2_1.Out
     Upshift=@TriggerAnyTriggerAny2_3.Out
    )."UE.OutputFormat.Mono.Audio:0"
    Gain 1=@DecibelstoLinearGain_1.Linear Gain
    In 1=&/MSP_EngineAudio_Distortion(
     Waveshaper.Amount=^Exhaust.Throttle.Waveshaper.Amount
     OffThrottleDistortionScale=^Exhaust.Throttle.Waveshaper.AmountOffScale
     OnThrottleDistortionScale=^Exhaust.Throttle.Waveshaper.AmountOnScale
     Waveshaper.Bias=^Exhaust.Throttle.Waveshaper.Bias
     Waveshaper.Bypass=^Exhaust.Throttle.Waveshaper.Bypass
     Waveshaper.Crossover=^Exhaust.Throttle.Waveshaper.Crossover
     Waveshaper.Mix=^Exhaust.Throttle.Waveshaper.Mix
     Waveshaper.OutputGain=^Exhaust.Throttle.Waveshaper.OutputGain
     Waveshaper.Type=^Exhaust.Throttle.Waveshaper.Type
     Input=&DM_EngineAudio_ThrottleVolumeFilter(
      Filter.FilterQ=^Exhaust.Throttle.EQ.FilterQ
      Filter.FilterType=^Exhaust.Throttle.EQ.FilterType
      Filter.OffThrottleFreq=^Exhaust.Throttle.EQ.OffFreq
      Filter.OffThrottleGain=^Exhaust.Throttle.EQ.OffGain
      Filter.OnThrottleFreq=^Exhaust.Throttle.EQ.OnFreq
      Filter.OnThrottleGain=^Exhaust.Throttle.EQ.OnGain
      Interp Time=^Exhaust.Throttle.InterpTime
      Volume.OffThrottle=^Exhaust.Throttle.Volume.Off
      Volume.OnThrottle=^Exhaust.Throttle.Volume.On
      AudioIn=@REVRevPlayer_0.Channel 0
      REV.Throttle=@n6.Value
     ).AudioOut
     REV.Throttle=@n9.Value
    ).Input
   )
   n10 = MapRange.MapRange:Float(In=^DriftAngle) {also {defaults {In Range B=30, Out Range B=-3}}}
   TriggerAnyTriggerAny3_0 = "TriggerAny.Trigger Any (3)"(In 0=^Downshift, In 1=^Upshift)
   n11 = VariableMutator.Trigger(Variable=InitVariable.Trigger().Variable, Value=@TriggerAnyTriggerAny3_0.Out)
   n12 = VariableDeferredAccessor.Float(Variable=InitVariable.Float().Variable)
   DecibelstoLinearGain_2 = UE.Decibels to Linear Gain:Float(Decibels=^REVVolume)
   n13 = /FortMetasound/Patches/Multiband_StereoWidener(Crossover=^Interior.StereoEffect.Crossover, MixLevel=^Interior.StereoEffect.Mix, Input=@REVRevPlayer_0.Channel 2) {also {defaults {Delay Time=0.072}}}
   n14 = &StereoBiquadFilter(Cutoff Frequency=^Interior.EQ.CutoffFrequency, Type=^Interior.EQ.FilterType, In L=@n13.Out L, In R=@n13.Out R)
   n15 = /FortMetasound/Patches/StereoMixer_One("Volume(Lin)"=@DecibelstoLinearGain_2.Linear Gain, InL=@n14.Out L, InR=@n14.Out R)
   n16 = &/DM_EngineAudio_GameSpecificMixer(
    Engine=@AudioMixerAudioMixerMono2_0.Out
    Exhaust=@AudioMixerAudioMixerMono2_1.Out
    Transmission=&/DM_EngineAudio_Transmision_Waveasset(
     grounded=^Fort.Vehicle.IsGrounded
     GearChange_Time=^GearChangeTime
     Transmission.MaxFreq=^Transmission.MaxFreq
     Transmission.MinFreq=^Transmission.MinFreq
     Transmission_OffThrottle=^Transmission.Throttle.Volume.Off
     Transmission_OnThrottle=^Transmission.Throttle.Volume.On
     TransmissionVolume=UE.Add:Float(AdditionalOperands=^Transmission.Volume.Overall, PrimaryOperand=@n10.Out Value).Out
     UE.Source.OnPlay=TriggerCompare.Clamp:Bool(A=^Transmission.Enabled, Compare=^UE.Source.OnPlay) {also {defaults {B="<MetaSoundBuilderResult.SUCCEEDED: 0>"}}}.True
     REV.Gear=VariableAccessor.Int32(Variable=@n8.Variable).Value
     GearChanged=VariableAccessor.Trigger(Variable=@n11.Variable).Value
     Speed_Normalized=@n12.Value
     REV.Throttle=@n5.Value
    )."UE.OutputFormat.Mono.Audio:0"
    InteriorL=@n15.OutL
    InteriorR=@n15.OutR
   )
   n17 = &/DM_EngineAudio_VolumeMixer(
    MixP=^Mix.MixPerspective
    MixP_Exterior=^Mix.Perspective.Exterior.Volume
    MixP_Interior=^Mix.Perspective.Interior.Volume
    UE.Source.OnPlay=^UE.Source.OnPlay
    EngineL=@n16.EngineL
    EngineR=@n16.EngineR
    ExhaustL=@n16.ExhaustL
    ExhaustR=@n16.ExhaustR
    InteriorL=@n16.InteriorL
    InteriorR=@n16.InteriorR
    Acceleration=Max.Max:Float(A=^Acceleration, B=^Fort.Vehicle.Acceleration).Value
    BonusSpeed=Max.Max:Float(A=^BonusSpeed, B=^Fort.Vehicle.BonusSpeed).Value
   ) {also {defaults {Volume_Engine=0, Volume_Exhaust=0, Volume_Interior=0, Volume_REV=0}}}
   n18 = &/DM_EngineAudio_MixPFilters(
    MixP=^Mix.MixPerspective
    MixP.Exterior.Engine.EQ=^Mix.Perspective.Exterior.Engine.EQ
    MixP.Exterior.Exhaust.EQ=^Mix.Perspective.Exterior.Exhaust.EQ
    MixP.Interior.Engine.EQ=^Mix.Perspective.Interior.Engine.EQ
    MixP.Interior.Exhaust.EQ=^Mix.Perspective.Interior.Exhaust.EQ
    UE.Source.OnPlay=^UE.Source.OnPlay
    EngineL=@n17.EngineL
    EngineR=@n17.EngineR
    ExhaustL=@n17.ExhaustL
    ExhaustR=@n17.ExhaustR
    InteriorL=@n17.InteriorL
    InteriorR=@n17.InteriorR
   )
   n19 = &/DM_EngineAudio_AzimuthPanner(
    EnginefilterType=^Mix.Azimuth.Engine.FilterType
    Engine_MaxFreq=^Mix.Azimuth.Engine.Freq.Max
    Engine_MinFreq=^Mix.Azimuth.Engine.Freq.Min
    Engine_MinVolume=^Mix.Azimuth.Engine.MinVolume
    ExhaustFilterType=^Mix.Azimuth.Exhaust.FilterType
    Exhaust_MaxFreq=^Mix.Azimuth.Exhaust.Freq.Max
    Exhaust_MinFreq=^Mix.Azimuth.Exhaust.Freq.Min
    Exhaust_MinVolume=^Mix.Azimuth.Exhaust.MinVolume
    EngineMute=^Mix.Mute.Engine
    ExhaustMute=^Mix.Mute.Exhaust
    InteriorMute=^Mix.Mute.Interior
    UE.Source.Orientation.Azimuth=^UE.Source.Orientation.Azimuth
    EngineL=@n18.EngineL
    EngineR=@n18.EngineR
    ExhaustL=@n18.ExhaustL
    ExhaustR=@n18.ExhaustR
    InteriorL=@n18.InteriorL
    InteriorR=@n18.InteriorR
   )
   DM_EngineAudio_PrimaryFX_0 = &/DM_EngineAudio_PrimaryFX(
    EQ1_FilterType=^Mix.EQ1.FilterType
    EQ1_Freq=^Mix.EQ1.Freq
    EQ1_Gain=^Mix.EQ1.Gain
    EQ1_Q=^Mix.EQ1.Q
    EQ2_FilterType=^Mix.EQ2.FilterType
    EQ2_Freq=^Mix.EQ2.Freq
    EQ2_Gain=^Mix.EQ2.Gain
    EQ2_Q=^Mix.EQ2.Q
    EQ3_FilterType=^Mix.EQ3.FilterType
    EQ3_Freq=^Mix.EQ3.Freq
    EQ3_Gain=^Mix.EQ3.Gain
    EQ3_Q=^Mix.EQ3.Q
    EQ4_FilterType=^Mix.EQ4.FilterType
    EQ4_Freq=^Mix.EQ4.Freq
    EQ4_Gain=^Mix.EQ4.Gain
    EQ4_Q=^Mix.EQ4.Q
    UE.Source.OnPlay=^UE.Source.OnPlay
    AudioIn_L=@n19.Out L
    AudioIn_R=@n19.Out R
   )
   VariableMutator.Float(Value=Max.Max:Float(B=^Fort.Vehicle.NormalSpeed, A=^NormalizedSpeed).Value, Variable=@n12.Variable)
  ]
  links [
   {dst="@UE.OutputFormat.Stereo.Audio:0", src=DM_EngineAudio_PrimaryFX_0.UE.OutputFormat.Stereo.Audio:0}
   {dst="@UE.OutputFormat.Stereo.Audio:1", src=DM_EngineAudio_PrimaryFX_0.UE.OutputFormat.Stereo.Audio:1}
  ]
 }]
properties {
 assetImportData=&self:AssetImportData
 bLooping=true
 duration=10000
 numChannels=2
 outputFormat=Stereo
 qualitySetting=Vehicles_REV_3CH
 sampleRate=48000
 soundClassObject=&Fort_SFX_InGame
 modulationSettings {
  pitchRouting=Union
  volumeRouting=Union
  pitchModulationDestination {bEnableModulation=true, modulators [&DM_Bus_Vehicle_Engine_Pitch]}
  volumeModulationDestination {bEnableModulation=true, modulators [&DM_VehicleEnginePatch]}
 }
}
