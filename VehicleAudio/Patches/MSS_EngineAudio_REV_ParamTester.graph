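# MetaSound source that simulates a looping drive cycle for testing engine audio parameters, outputting RPM, Throttle, SpeedNormalized, Gear, GearChange and ExhaustPop with no audio connected.
# While Demo runs, per-gear wavetable envelopes supply RPM, each finish advancing the gear until a limit resets it and waits IdleTime.
# Speed follows interpolated gear progress, while throttle is inferred by threshold-comparing RPM against its deferred-variable feedback and smoothing the result.
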MetaSoundSource /VehicleAudio/Patches/MSS_EngineAudio_REV_ParamTester
inputs [
 Demo {type=Bool}
 gearchangetime {type=Time}
 IdleTime {type=Time}
 RPM_WaveTableBank {type=WaveTableBankAsset}
 UE.Source.OnPlay {type=Trigger}
]
interfaces [UE.OutputFormat.Mono:1.0, UE.Source:1.1]
outputs [
 ExhaustPop {type=Trigger}
 Gear {type=Int32}
 GearChange {type=Trigger}
 RPM {type=Float}
 SpeedNormalized {type=Float}
 Throttle {type=Float}
 "UE.OutputFormat.Mono.Audio:0" {type=Audio}
]
pages [Default {
  input_defaults {
   Demo=null
   IdleTime=5
   RPM_WaveTableBank=&/DM_EngineAudio_ParamPreviewer_RPM
   UE.Source.OnPlay=null
   gearchangetime=0.896
  }
  nodes [
   n1 = InitVariable.Int32()
   n2 = VariableDeferredAccessor.Int32(Variable=@n1.Variable)
   TriggerOnValueChange_2 = UE.Trigger On Value Change:Int32(Value=@n2.Value)
   TriggerAnyTriggerAny2_3 = "TriggerAny.Trigger Any (2)"(In 1=^UE.Source.OnPlay, In 0=@TriggerOnValueChange_2.Trigger)
   n3 = Array.Get:Float:Array(
    Trigger=@TriggerAnyTriggerAny2_3.Out
    Array=VariableAccessor.Float:Array(Variable=InitVariable.Float:Array().Variable).Value
    Index=@n2.Value
   )
   n4 = UE.ConversionFloatToTime(In=@n3.Element)
   UE.ConversionTimeToFloat(In=@n4.Out)
   InitVariable.Bool()
   InitVariable.Bool()
   InitVariable.Float()
   n5 = VariableMutator.Float(Value=@n3.Element, Variable=InitVariable.Float().Variable)
   n6 = UE.ConversionTimeToFloat(In=^gearchangetime)
   n7 = VariableDeferredAccessor.Int32(Variable=@n1.Variable)
   n8 = VariableDeferredAccessor.Int32(Variable=@n1.Variable)
   TriggerOnValueChange_3 = UE.Trigger On Value Change:Int32(Value=@n8.Value)
   n9 = VariableAccessor.Int32(Variable=InitVariable.Int32().Variable)
   n10 = TriggerCompare.Clamp:Int32(Compare=@TriggerOnValueChange_3.Trigger, B=@n9.Value, A=@n8.Value)
   TriggerOnValueChange_0 = UE.Trigger On Value Change:Bool(Value=^Demo)
   TriggerAnyTriggerAny2_4 = "TriggerAny.Trigger Any (2)"(In 1=^UE.Source.OnPlay, In 0=@TriggerOnValueChange_0.Trigger)
   n11 = TriggerCompare.Clamp:Bool(A=^Demo, Compare=@TriggerAnyTriggerAny2_4.Out) {also {defaults {B="<MetaSoundBuilderResult.SUCCEEDED: 0>"}}}
   n12 = VariableAccessor.Trigger(Variable=VariableMutator.Trigger(Variable=InitVariable.Trigger().Variable, Value=@n11.True).Variable)
   n13 = VariableDeferredAccessor.Trigger(Variable=InitVariable.Trigger().Variable)
   n14 = VariableDeferredAccessor.Trigger(Variable=@n13.Variable)
   n15 = VariableAccessor.Trigger(Variable=VariableMutator.Trigger(Value=@n11.False, Variable=@n14.Variable).Variable)
   TriggerAnyTriggerAny3_1 = "TriggerAny.Trigger Any (3)"(In 2=@n10.True, In 0=@n12.Value, In 1=@n15.Value)
   n16 = Convert.Float:Int32(Int32=VariableDeferredAccessor.Int32(Variable=@n1.Variable).Value)
   n17 = VariableAccessor.Trigger(Variable=VariableMutator.Trigger(Variable=InitVariable.Trigger().Variable, Value=@n10.True).Variable)
   TriggerAnyTriggerAny2_5 = "TriggerAny.Trigger Any (2)"(In 0=@n17.Value, In 1=@n14.Value)
   TriggerDelay_2 = UE.Trigger Delay(Delay Time=^IdleTime, In=@n17.Value)
   n18 = VariableDeferredAccessor.Trigger(Variable=InitVariable.Trigger().Variable)
   TriggerControl_2 = UE.Trigger Control(Open=@TriggerDelay_2.Out, Close=@n17.Value, Trigger In=@n18.Value)
   TriggerDelay_4 = UE.Trigger Delay(
    Delay Time=UE.Divide:Time by Float(PrimaryOperand=^gearchangetime) {also {defaults {AdditionalOperands=2}, version [1,1]}}.Out
    In=@TriggerControl_2.Trigger Out
   )
   n19 = VariableAccessor.Trigger(Variable=@n12.Variable)
   TriggerAnyTriggerAny3_0 = "TriggerAny.Trigger Any (3)"(In 1=@TriggerDelay_2.Out, In 0=@TriggerDelay_4.Out, In 2=@n19.Value)
   n20 = UE.WaveTableEnvelope(
    Duration=@n4.Out
    WaveTable=UE.MetasoundWaveTableGet(WaveTableBank=^RPM_WaveTableBank, TableIndex=UE.Add:Float(PrimaryOperand=@n16.Float).Out) {also {version [1,1]}}.Out
    Stop=@TriggerAnyTriggerAny2_5.Out
    Play=@TriggerAnyTriggerAny3_0.Out
   )
   ValueValue_1 = Value.Value:Int32(
    Target Value=UE.Add:Int32(PrimaryOperand=VariableDeferredAccessor.Int32(Variable=@n7.Variable).Value) {also {defaults {AdditionalOperands=1}}}.Out
    Reset=@TriggerAnyTriggerAny3_1.Out
    Set=@n20.OnFinished
   ) {also {defaults {Init Value=0}}}
   n21 = VariableAccessor.Int32(Variable=VariableMutator.Int32(Value=@ValueValue_1.Output Value, Variable=@n2.Variable).Variable)
   n22 = UE.Add:Float(PrimaryOperand=Convert.Float:Int32(Int32=VariableAccessor.Int32(Variable=@n21.Variable).Value).Float) {also {defaults {AdditionalOperands=1}}}
   n23 = Convert.Float:Int32(Int32=VariableAccessor.Int32(Variable=@n9.Variable).Value)
   n24 = VariableAccessor.Trigger(Variable=@n15.Variable)
   n25 = VariableAccessor.Trigger(Variable=@n17.Variable)
   TriggerAnyTriggerAny2_0 = "TriggerAny.Trigger Any (2)"(In 0=VariableAccessor.Trigger(Variable=@n24.Variable).Value, In 1=@n25.Value)
   TriggerOnValueChange_1 = UE.Trigger On Value Change:Int32(Value=@n21.Value)
   TriggerDelay_3 = UE.Trigger Delay(Delay Time=^gearchangetime, In=@TriggerOnValueChange_1.Trigger)
   TriggerAnyTriggerAny2_2 = "TriggerAny.Trigger Any (2)"(In 1=^UE.Source.OnPlay, In 0=@TriggerDelay_3.Out)
   TriggerDelay_0 = UE.Trigger Delay(Delay Time=^IdleTime, In=VariableAccessor.Trigger(Variable=@n25.Variable).Value)
   TriggerControl_0 = UE.Trigger Control(Close=@TriggerAnyTriggerAny2_0.Out, Trigger In=@TriggerAnyTriggerAny2_2.Out, Open=@TriggerDelay_0.Out)
   InterpSpeed_0 = &InterpSpeed(
    RisingInterpSpeed=UE.Add:Float(AdditionalOperands=UE.ConversionTimeToFloat(In=^gearchangetime).Out, PrimaryOperand=VariableAccessor.Float(Variable=@n5.Variable).Value).Out
    FallingInterpSpeed=@n6.Out
    FloatInput=Value.Value:Float(
     Target Value=UE.Divide:Float(PrimaryOperand=@n22.Out, AdditionalOperands=@n23.Float).Out
     Reset=@TriggerAnyTriggerAny2_0.Out
     Set=@TriggerControl_0.Trigger Out
    ).Output Value
   )
   n26 = &InterpSpeed(FallingInterpSpeed=@n6.Out, FloatInput=@n20.Out) {also {defaults {RisingInterpSpeed=0}}}
   UE.Multiply:Float(PrimaryOperand=UE.Multiply:Float(PrimaryOperand=@n26.InterpValue).Out)
   TriggerAnyTriggerAny2_1 = "TriggerAny.Trigger Any (2)"(In 0=@n10.True, In 1=@n24.Value)
   TriggerDelay_1 = UE.Trigger Delay(Delay Time=^gearchangetime, In=@TriggerAnyTriggerAny2_3.Out)
   TriggerOnValueChange_4 = UE.Trigger On Value Change:Int32(Value=@n7.Value)
   n27 = InitVariable.Float()
   n28 = UE.TriggerOnThreshold:Float(In=@n26.FloatOutput, Threshold=VariableDeferredAccessor.Float(Variable=@n27.Variable).Value)
   n29 = VariableDeferredAccessor.Float(Variable=@n27.Variable)
   n30 = UE.TriggerOnThreshold:Float(In=@n26.FloatOutput, Threshold=@n29.Value) {also {defaults {Type=1}}}
   TriggerOnce_0 = UE.Trigger Once(Reset=@n28.Out, Trigger In=@n30.Out)
   TriggerOnce_1 = UE.Trigger Once(Trigger In=@n28.Out, Reset=@n30.Out)
   TriggerRouteTriggerRouteFloat2_0 = "TriggerRoute.Trigger Route (Float, 2):Float"(Set 0=@TriggerOnce_0.Trigger Out, Set 1=@TriggerOnce_1.Trigger Out) {also {defaults {Value 1=1}}}
   InterpSpeed_2 = &InterpSpeed(FloatInput=@TriggerRouteTriggerRouteFloat2_0.Value) {also {defaults {FallingInterpSpeed=0.1, RisingInterpSpeed=0.1}}}
   VariableMutator.Float(Variable=InitVariable.Float().Variable, Value=@InterpSpeed_2.FloatOutput)
   Multiply_3 = UE.Multiply:Float(PrimaryOperand=UE.Multiply:Float(PrimaryOperand=@n26.FloatOutput).Out)
   VariableMutator.Float(Value=@Multiply_3.Out, Variable=@n29.Variable)
   TriggerControl_1 = UE.Trigger Control(Open=VariableAccessor.Trigger(Variable=@n19.Variable).Value, Close=@n13.Value, Trigger In=@n20.OnFinished)
   VariableMutator.Trigger(Value=@TriggerControl_1.Trigger Out, Variable=@n18.Variable)
  ]
  links [
   {dst="@SpeedNormalized", src=InterpSpeed_0.FloatOutput}
   {dst="@Throttle", src=InterpSpeed_2.FloatOutput}
   {dst="@RPM", src=Multiply_3.Out}
   {dst="@GearChange", src=TriggerOnValueChange_4.Trigger}
   {dst="@ExhaustPop", src=ValueValue_1.On Set}
   {dst="@Gear", src=ValueValue_1.Output Value}
  ]
 }]
properties {
 assetImportData=&self:AssetImportData
 bLooping=true
 duration=10000
 sampleRate=48000
 volume=2
}
