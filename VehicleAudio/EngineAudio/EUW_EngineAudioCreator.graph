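# Editor utility widget that creates vehicle engine audio assets from selected REV models and SoundWaves.
# After budget checks, MetaSound builders fill preset inputs, matching waves by name and models by sample rate (Mobile page), then save engine, ignition, off and horn presets.
# It also creates REV settings, filled from .model XML via embedded Python, and engine audio item definitions.
# Preview controls spawn sounds in the viewport, watch RPM/throttle/gear outputs and toggle a vehicle mix.
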EditorUtilityWidgetBlueprint /VehicleAudio/EngineAudio/EUW_EngineAudioCreator
animations [
 OnRevSettingsChanged_Success {
  binding_count=2
  length_seconds=1.5
  object=&self:OnRevSettingsChanged_Success
  sequence {
   bindings [
    REVSettingsBackground_Green {
     id=#1
     parent_id=null
     possessed_class=Image
     tracks [{
       type=MovieSceneColorTrack
       display_name=Color and Opacity
       sections [{
         type=MovieSceneColorSection
         end_frame=30
         start_frame=0
         channels [
          Color.R {type=MovieSceneScriptingFloatChannel, const [0.0026,1]}
          Color.G {type=MovieSceneScriptingFloatChannel, const [1,1]}
          Color.B {type=MovieSceneScriptingFloatChannel, const [0.1925,1]}
          Color.A {type=MovieSceneScriptingFloatChannel, keys [[0,0],[1,1],[30,0]]}
         ]
        }]
      }]
    }
    REVSettingsBackground_Red {
     id=#2
     parent_id=null
     possessed_class=Image
     tracks [{
       type=MovieSceneColorTrack
       display_name=Color and Opacity
       sections [{
         type=MovieSceneColorSection
         end_frame=0
         start_frame=0
         channels [
          Color.R {type=MovieSceneScriptingFloatChannel, const [1,1]}
          Color.G {type=MovieSceneScriptingFloatChannel, const [0,1]}
          Color.B {type=MovieSceneScriptingFloatChannel, const [0.0343,1]}
          Color.A {type=MovieSceneScriptingFloatChannel, const [0,1]}
         ]
        }]
      }]
    }
   ]
   display_rate {denominator=1, numerator=20}
   tick_resolution {denominator=1, numerator=60000}
   tracks []
  }
 }
 OnRevSettingsChanged_Failure {
  binding_count=2
  length_seconds=1.5
  object=&self:OnRevSettingsChanged_Failure
  sequence {
   bindings [
    REVSettingsBackground_Green {
     id=#3
     parent_id=null
     possessed_class=Image
     tracks [{
       type=MovieSceneColorTrack
       display_name=Color and Opacity
       sections [{
         type=MovieSceneColorSection
         end_frame=0
         start_frame=0
         channels [
          Color.R {type=MovieSceneScriptingFloatChannel, const [0.0026,1]}
          Color.G {type=MovieSceneScriptingFloatChannel, const [1,1]}
          Color.B {type=MovieSceneScriptingFloatChannel, const [0.1925,1]}
          Color.A {type=MovieSceneScriptingFloatChannel, const [0,1]}
         ]
        }]
      }]
    }
    REVSettingsBackground_Red {
     id=#4
     parent_id=null
     possessed_class=Image
     tracks [{
       type=MovieSceneColorTrack
       display_name=Color and Opacity
       sections [{
         type=MovieSceneColorSection
         end_frame=30
         start_frame=0
         channels [
          Color.R {type=MovieSceneScriptingFloatChannel, const [1,1]}
          Color.G {type=MovieSceneScriptingFloatChannel, const [0,1]}
          Color.B {type=MovieSceneScriptingFloatChannel, const [0.0343,1]}
          Color.A {type=MovieSceneScriptingFloatChannel, keys [[0,0],[1,1],[30,0]]}
         ]
        }]
      }]
    }
   ]
   display_rate {denominator=1, numerator=20}
   tick_resolution {denominator=1, numerator=60000}
   tracks []
  }
 }
]
graphs [
 CheckIfAssetExists {nodes [
   n1 = CheckIfAssetExists!() -> (then, AssetName:String, AssetPath:String) {is=title K2Node_FunctionEntry}
   > n2 = DoesAssetExist(AssetPath=Append!(A=@n1.AssetPath, B=/, C:String=@n1) -> (ReturnValue))
   > n3 = Branch(Condition=@n2)
   > Show Message Dialog(
    Title="To Text (String)"(InString=Error:)
    Message="To Text (String)"(InString=Append(A="The Asset ", B=@n1, C=" already exists, so should not be re-created, so it will not be created. "))
    MessageType=Ok
    DefaultValue=Ok
    MessageCategory=Error
   )
   > Return Node!(execute, ReturnValue:bool=@n2)
   Return Node(execute=@n3.else, ReturnValue=@n2)
   n4 = Get AssetNameInput() -> (AssetNameInput:Editor Utility Editable Text Box Object Reference)
   "Equal (Boolean)"(A:bool=TextIsEmpty(InText=Get Text(self=@n4)), B:bool) -> (ReturnValue:bool=~) {is=title K2Node_PromotableOperator}
  ]}
 CheckREVModelSize {nodes [
   n1 = CheckREVModelSize!() -> (then, MonoREVSizeLimit:int=1800, 3CHREVSizeLimit:int) {is=title K2Node_FunctionEntry}
   > Set bREVOverBudget(bREVOverBudget:bool=false) -> (Output_Get:bool)
   > n2 = GetSelectedAssetsOfClass(AssetClass=/Script/REVRuntime.REVModel)
   > n3 = For Each Loop(Array:Array of Object References=@n2) -> (Array Element:Object Reference)
   Return Node!(execute=@n3.Completed, bREVOverBudget:bool=Get bREVOverBudget() -> (bREVOverBudget:bool))
   > n4 = GetEditorProperty!(execute:Exec, Object:Object Reference=@n3, PropertyName:Name=NumChannels) -> (then:Exec, PropertyValue:int, ReturnValue:bool)
   > n5 = Switch on Int!(execute:Exec, Selection:int=@n4) -> (Default:Exec, 0:Exec, 1:Exec, 2:Exec, 3:Exec)
   n6 = GetEditorProperty(execute=@n5.1, Object=@n3, PropertyName=AssetSizeKB)
   > Branch(Condition="integer > integer"(A=@n6, B=@n1))
   n7 = float / float(A="To Float (Integer)"(InInt=@n6), B=1000)
   n8 = "To String (Float)"(InDouble=@n7)
   n9 = float / float(A="To Float (Integer)"(InInt=@n1), B=1000)
   n10 = "To String (Float)"(InDouble=@n9)
   n11 = "To String (Float)"(InDouble=float - float(A=@n7, B=@n9))
   > n12 = Show Message Dialog(
    Title="To Text (String)"(InString=Error:)
    Message="To Text (String)"(InString=Append!(
      A="Your Mono Rev Model is over size budget. Double check your ramp and idle loop to be sure they are trimmed as much as possible to be within "
      B=@n10
      C="MB of size. "
      D:String=Your file is
      E:String=@n8
      F:String="MB,"
      G:String=@n11
      H:String=" MB over budget."
     ) -> (ReturnValue))
    MessageType=Ok
    DefaultValue=No
    MessageCategory=Error
   )
   Set bREVOverBudget(execute=[@n12.then, @n15.then], bREVOverBudget=true)
   n13 = GetEditorProperty(execute=@n5.3, Object=@n3, PropertyName=AssetSizeKB)
   > Branch(Condition="integer > integer"(A=@n13, B=@n1.3CHREVSizeLimit))
   n14 = "To String (Float)"(InDouble=float / float(A="To Float (Integer)"(InInt=@n1.3CHREVSizeLimit), B=1000))
   > n15 = Show Message Dialog(
    Title="To Text (String)"(InString=Error:)
    Message="To Text (String)"(InString=Append!(
      A="Your 3CH Rev Model is over size budget. Double check your ramp and idle loop to be sure they are trimmed as much as possible to be within "
      B=@n14
      C=MB of size
     ) -> (ReturnValue))
    MessageType=Ok
    DefaultValue=No
    MessageCategory=Error
   )
  ]}
 CreateTimelineAnimationControllers {nodes [
   Inputs()
   > n1 = Sequence()
   > n2 = Create Timeline Animation Controller!(
    execute
    WidgetTarget:User Widget Object Reference
    Animation:Widget Animation Object Reference=Get OnRevSettingsChanged_Success() -> (OnRevSettingsChanged_Success:Widget Animation Object Reference)
   ) -> (then, ReturnValue:Timeline Animation Controller Object Reference)
   > Set RevSettingsSuccessAnimation(RevSettingsSuccessAnimation:Timeline Animation Controller Object Reference=@n2) -> (Output_Get:Timeline Animation Controller Object Reference)
   Sequence(execute=@n1.then_1)
   > n3 = Create Timeline Animation Controller(Animation=Get OnRevSettingsChanged_Failure() -> (OnRevSettingsChanged_Failure:Widget Animation Object Reference))
   > Set RevSettingsFailureAnimation(RevSettingsFailureAnimation:Timeline Animation Controller Object Reference=@n3) -> (Output_Get:Timeline Animation Controller Object Reference)
   > Outputs!(then:Exec)
  ]}
 EventGraph {nodes [
   "On Pressed (EditorUtilityButton_56_MakeFolder)"() {is=title K2Node_ComponentBoundEvent}
   > n1 = Branch(Condition="Equal Exactly (String)"(A=Get SanitizedAssetName() -> (SanitizedAssetName:String)))
   Show Message Dialog(
    execute=[@n6.else, @n1.then]
    Title="To Text (String)"(InString=Error:)
    Message="To Text (String)"(InString="No asset name input selected. Please input an asset name. ")
    MessageType=Ok
    DefaultValue=No
    MessageCategory=Error
   )
   n2 = GetCurrentContentBrowserPath!() -> (OutPath:String, ReturnValue:bool)
   n3 = Append(A="To String (Text)"(InText="To Text (String)"(InString=@n2)), B=/, C=Get SanitizedAssetName())
   MakeDirectory!(
    execute=@n1.else
    self:Editor Asset Subsystem Object Reference=Get EditorAssetSubsystem() -> (ReturnValue:Editor Asset Subsystem Object Reference) {is=title K2Node_GetEditorSubsystem}
    DirectoryPath:String=@n3
   ) -> (then, ReturnValue:bool)
   > n4 = Set FolderDirectory(FolderDirectory:String=@n3) -> (Output_Get:String)
   > SyncBrowserToFolders!(execute, FolderList:Array of Strings=Make Array("[0]":String=@n4) -> (Array:Array of Strings)) -> (then)
   "On Pressed (EditorUtilityButton_56_MakeBoost)"() {is=title K2Node_ComponentBoundEvent}
   n5 = Get Text(self=Get AssetNameInput())
   > n6 = Branch(Condition="Equal (Boolean)"(A=TextIsEmpty(InText=@n5)) -> (ReturnValue=~))
   > n7 = GetSelectedAssetsOfClass(AssetClass=&REVModel)
   > n8 = Branch(Condition=Is Not Empty!(TargetArray:Array of Object References=@n7) -> (ReturnValue:bool))
   > n9 = CheckREVModelSize!(execute, self, MonoREVSizeLimit:int=1899, 3CHREVSizeLimit:int=6000) -> (then, bREVOverBudget:bool) {is=title K2Node_CallFunction}
   > n10 = Branch(Condition=@n9)
   n11 = Sequence!(execute=@n10.else) -> (then_0, then_1, then_2:Exec, then_3:Exec, then_4:Exec, then_5:Exec, then_6:Exec, then_7:Exec)
   > n12 = GenerateMetasoundBuilder!(execute, self, Object:Object Reference=Get LowSpecEngineAudioMSS() -> (LowSpecEngineAudioMSS:Object Reference)) -> (then, Source Preset Builder:Meta Sound Source Builder Object Reference)
   > Set LowSpecEngineAudioBuilder(LowSpecEngineAudioBuilder:Meta Sound Source Builder Object Reference=@n12) -> (Output_Get:Meta Sound Source Builder Object Reference)
   > n13 = GenerateMetasoundBuilder(Object=Get Ignition Preset MSS() -> (Ignition Preset MSS:Object Reference))
   > Set Ignition Preset MSS Builder(Ignition Preset MSS Builder:Meta Sound Source Builder Object Reference=@n13) -> (Output_Get:Meta Sound Source Builder Object Reference)
   > n14 = GenerateMetasoundBuilder(Object=Get Off Preset MSS() -> (Off Preset MSS:Object Reference))
   > Set Off Preset MSS Builder(Off Preset MSS Builder:Meta Sound Source Builder Object Reference=@n14) -> (Output_Get:Meta Sound Source Builder Object Reference)
   > n15 = GetSelectedAssetsOfClass(AssetClass=/Script/Engine.SoundWave)
   > n16 = For Each Loop(Array=@n15)
   > Branch(Condition=Contains(SearchIn="To String (Object)"!(InObj:Object Reference=@n16) -> (ReturnValue:String), Substring=horn))
   > Do Once()
   > n17 = GenerateMetasoundBuilder(Object=Get Horn Preset MSS() -> (Horn Preset MSS:Object Reference))
   > Set Horn Preset MSS Builder(Horn Preset MSS Builder:Meta Sound Source Builder Object Reference=@n17) -> (Output_Get:Meta Sound Source Builder Object Reference)
   n18 = GetSelectedAssetsOfClass(execute=@n11.then_1, AssetClass=&SoundWave)
   > n19 = For Each Loop(Array=@n18)
   > Branch(Condition=Contains(SearchIn=GetObjectName(Object=@n19), Substring=Horn))
   > n20 = Create MetaSound Literal from Object!(execute, Value:Object Reference=@n19) -> (then, ReturnValue:Meta Sound Literal Structure)
   > SetGraphInputDefault!(
    execute
    self:Meta Sound Builder Base Object Reference=Get Horn Preset MSS Builder() -> (Horn Preset MSS Builder:Meta Sound Source Builder Object Reference)
    InputName:Name=Wave Asset
    Literal:"Meta Sound Literal Structure (by ref)"=@n20
    PageName:Name
   ) -> (Succeeded:Exec, Failed:Exec)
   n21 = GetSelectedAssetsOfClass(execute=@n11.then_3, AssetClass=&REVModel)
   > Branch(Condition=Is Not Empty(TargetArray=@n21))
   > n22 = For Each Loop(Array=@n21)
   > n23 = GetEditorProperty(Object=@n22, PropertyName=NumChannels)
   > n24 = Branch(Condition="Equal (Integer)"(A=@n23, B=1))
   > n25 = GetEditorProperty(Object=@n22, PropertyName=SampleRate) -> (PropertyValue:double)
   > n26 = Branch(Condition="Equal (Float)"(A=@n25, B=Get RevFullResolutionSampleRate() -> (RevFullResolutionSampleRate:double)))
   > n27 = Create MetaSound Literal from Object(Value=@n22)
   n28 = Get LowSpecEngineAudioBuilder() -> (LowSpecEngineAudioBuilder:Meta Sound Source Builder Object Reference)
   > SetGraphInputDefault(self=@n28, InputName=Mono, Literal=@n27)
   n29 = Branch(execute=@n26.else, Condition="Equal (Float)"(A=@n25, B=Get RevMobileResolutionSampleRate() -> (RevMobileResolutionSampleRate:double)))
   > n30 = Create MetaSound Literal from Object(Value=@n22)
   > SetGraphInputDefault(self=@n28, InputName=Mono, Literal=@n30, PageName=Mobile)
   K2Node_Composite_1 = SampleRateError(execute=@n29.else) {is=title K2Node_Composite}
   K2Node_Composite_0 = ChannelCountWarning(execute=@n24.else) {is=title K2Node_Composite}
   n31 = GetSelectedAssetsOfClass(execute=@n11.then_4, AssetClass=&SoundWave)
   > n32 = For Each Loop(Array=@n31)
   n33 = Create MetaSound Object Array Literal!(
    execute=@n32.Completed
    self:Meta Sound Builder Subsystem Object Reference=Get MetaSoundBuilderSubsystem() -> (ReturnValue:Meta Sound Builder Subsystem Object Reference) {is=title K2Node_GetEngineSubsystem}
    Value:Array of Object References=Get IgnitionWaveArrayInput() -> (IgnitionWaveArrayInput:Array of Object References)
   ) -> (then, ReturnValue:Meta Sound Literal Structure)
   > n34 = SetGraphInputDefault(
    self=Get Ignition Preset MSS Builder() -> (Ignition Preset MSS Builder:Meta Sound Source Builder Object Reference)
    InputName=IgnitionSound
    Literal=@n33
   )
   n35 = Create MetaSound Object Array Literal(
    execute=[@n34.Failed, @n34.Succeeded]
    self=Get MetaSoundBuilderSubsystem()
    Value=Get OffWaveArrayInput() -> (OffWaveArrayInput:Array of Object References)
   )
   > n36 = SetGraphInputDefault(self=Get Off Preset MSS Builder() -> (Off Preset MSS Builder:Meta Sound Source Builder Object Reference), InputName=In Array, Literal=@n35)
   Clear(execute=[@n36.Failed, @n36.Succeeded], TargetArray:Array of Object References=Get IgnitionWaveArrayInput())
   > Clear(TargetArray=Get OffWaveArrayInput())
   n37 = Branch(execute=@n32.LoopBody, Condition=Contains(SearchIn=GetObjectName(Object=@n32), Substring=Ignition))
   > Add(TargetArray:Array of Object References=Get IgnitionWaveArrayInput(), NewItem:"Object Reference (by ref)"=@n32) {is=title K2Node_CallArrayFunction}
   Branch(execute=@n37.else, Condition=Contains(SearchIn=GetObjectName(Object=@n32), Substring=Off, bSearchFromEnd=true))
   > Add(TargetArray=Get OffWaveArrayInput(), NewItem=@n32)
   n38 = GetSelectedAssetsOfClass(execute=@n11.then_5, AssetClass=/Script/REVRuntime.RevSettings)
   > n39 = Branch(Condition=Is Not Empty(TargetArray=@n38))
   > n40 = For Each Loop(Array=@n38)
   > n41 = Cast To RevSettings!(execute, Object=@n40) -> (then, CastFailed, AsRev Settings:Rev Settings Object Reference) {is=title K2Node_DynamicCast}
   > n42 = Set RevSettingsAsset(RevSettingsAsset:Rev Settings Object Reference=@n41) -> (Output_Get:Rev Settings Object Reference)
   n43 = Create MetaSound Literal from Object(execute=[@n42.then, @n51.then], Value=Get RevSettingsAsset() -> (RevSettingsAsset:Rev Settings Object Reference))
   > SetGraphInputDefault(self=Get LowSpecEngineAudioBuilder(), InputName=Settings, Literal=@n43)
   n44 = GetCurrentContentBrowserPath()
   n45 = Append!(A=REVSettings_, B=Get SanitizedAssetName()) -> (ReturnValue)
   n46 = CheckIfAssetExists!(execute=@n39.else, self, AssetName:String=@n45, AssetPath:String=@n44) -> (then, ReturnValue:bool) {is=title K2Node_CallFunction}
   > n47 = Branch(Condition=@n46)
   n48 = GetAssetTools(execute=@n47.else)
   > n49 = CreateAsset!(
    execute:Exec
    self:Object Reference="To Object (Interface)"!(Interface:"Interface Interface (by ref)"=@n48) -> (ReturnValue:Object Reference)
    AssetName:String=@n45
    PackagePath:String=@n44
    AssetClass:Object Class Reference=&RevSettings
    Factory:Factory Object Reference
    CallingContext:Name=None
    bOverwriteExisting:bool=false
   ) -> (then:Exec, ReturnValue:Object Reference) {is=title K2Node_Message}
   > SaveLoadedAsset(AssetToSave=@n49, bOnlyIfIsDirty=true)
   > n50 = Cast To RevSettings(Object=@n49)
   > n51 = Set RevSettingsAsset(RevSettingsAsset=@n50)
   n52 = Sequence!(execute=@n11.then_7) -> (then_0, then_1, then_2, then_3, then_4)
   n53 = GetCurrentContentBrowserPath()
   n54 = Append!(A=MSS_, B=EngineAudio_, C:String=Get SanitizedAssetName(), D:String=_LowSpec, E:String=_REV, F:String=_Preset) -> (ReturnValue)
   > n55 = CheckIfAssetExists(AssetName=@n54, AssetPath=@n53)
   > n56 = Branch(Condition=@n55)
   n57 = BuildToAsset!(
    execute=@n56.else
    self:Meta Sound Editor Subsystem Object Reference=Get MetaSoundEditorSubsystem() -> (ReturnValue:Meta Sound Editor Subsystem Object Reference) {is=title K2Node_GetEditorSubsystem}
    InBuilder:Meta Sound Builder Base Object Reference=Get LowSpecEngineAudioBuilder()
    Author:String=GetPlatformUserName!() -> (ReturnValue:String)
    AssetName:String=@n54
    PackagePath:String=@n53
    TemplateSoundWave:Sound Wave Object Reference=/VehicleAudio/Patches/MSS_EngineAudio_LowSpec_Parent_REV
   ) -> (Succeeded:Exec, Failed:Exec, ReturnValue:Meta Sound Document Interface Interface)
   > n58 = Cast To MetaSoundSource!(execute, Object:Meta Sound Document Interface Interface=@n57) -> (then, CastFailed, AsMetaSound Source:Meta Sound Source Object Reference) {is=title K2Node_DynamicCast}
   > Set Engine Sound(Engine Sound:Sound Base Soft Object Reference=Make Soft Reference!(Input:Meta Sound Source Object Reference=@n58) -> (Output:Meta Sound Source Soft Object Reference)) -> (Output_Get:Sound Base Soft Object Reference)
   n59 = GetCurrentContentBrowserPath()
   n60 = Append!(A=MSS_, B=EngineAudio_, C=Get SanitizedAssetName(), D=_Ignition, E=_Preset) -> (ReturnValue)
   n61 = CheckIfAssetExists(execute=@n52.then_1, AssetName=@n60, AssetPath=@n59)
   > n62 = Branch(Condition=@n61)
   n63 = BuildToAsset(
    execute=@n62.else
    self=Get MetaSoundEditorSubsystem()
    InBuilder=Get Ignition Preset MSS Builder()
    Author=GetPlatformUserName()
    AssetName=@n60
    PackagePath=@n59
    TemplateSoundWave=/VehicleAudio/Patches/MSS_EngineAudio_Ignition
   )
   > n64 = Cast To MetaSoundSource(Object=@n63)
   > Set Engine Ignition Sound(Engine Ignition Sound:Sound Base Object Reference=@n64) -> (Output_Get:Sound Base Object Reference)
   n65 = GetCurrentContentBrowserPath()
   n66 = Append(A=MSS_, B=EngineAudio_, C=Get SanitizedAssetName(), D=_Off, E=_Preset)
   n67 = CheckIfAssetExists(execute=@n52.then_2, AssetName=@n66, AssetPath=@n65)
   > n68 = Branch(Condition=@n67)
   n69 = BuildToAsset(
    execute=@n68.else
    self=Get MetaSoundEditorSubsystem()
    InBuilder=Get Off Preset MSS Builder()
    Author=GetPlatformUserName()
    AssetName=@n66
    PackagePath=@n65
    TemplateSoundWave=/VehicleAudio/Patches/MSS_EngineAudio_Off
   )
   > n70 = Cast To MetaSoundSource(Object=@n69)
   > Set Engine Off Sound(Engine Off Sound:Sound Base Object Reference=@n70) -> (Output_Get:Sound Base Object Reference)
   n71 = Get Horn Preset MSS Builder()
   Is Valid(exec=@n52.then_3, InputObject=@n71)
   n72 = GetCurrentContentBrowserPath()
   n73 = Append(A=MSS_, B=EngineAudio_, C=Get SanitizedAssetName(), D=_Horn, E=_Preset)
   > n74 = CheckIfAssetExists(AssetName=@n73, AssetPath=@n72)
   > n75 = Branch(Condition=@n74)
   n76 = BuildToAsset(
    execute=@n75.else
    self=Get MetaSoundEditorSubsystem()
    InBuilder=@n71
    Author=GetPlatformUserName()
    AssetName=@n73
    PackagePath=@n72
    TemplateSoundWave=/VehicleAudio/Patches/MSS_Vehicle_Horn
   )
   > n77 = Cast To MetaSoundSource(Object=@n76)
   > Set Horn Sound(Horn Sound:Sound Base Object Reference=@n77) -> (Output_Get:Sound Base Object Reference)
   n78 = GetCurrentContentBrowserPath()
   SaveDirectory!(execute=@n52.then_4, DirectoryPath:String=@n78, bOnlyIfIsDirty:bool=true, bRecursive:bool=false) -> (then, ReturnValue:bool)
   > SaveDirectory(DirectoryPath="To String (Boolean)"(InBool=@n78.ReturnValue), bOnlyIfIsDirty=true, bRecursive=false)
   n79 = Sequence!(execute=@n11.then_6) -> (then_0, then_1)
   > n80 = Create MetaSound Literal from Boolean!(execute, Value:bool=true) -> (then, ReturnValue:Meta Sound Literal Structure)
   > n81 = Sequence()
   n82 = Get LowSpecEngineAudioBuilder()
   > SetGraphInputDefault(self=@n82, InputName=REV.EnableShifting, Literal=@n80)
   SetGraphInputDefault(execute=@n81.then_1, self=@n82, InputName=Fort.Vehicle.Enableshifting, Literal=@n80)
   n83 = Create MetaSound Literal from Float!(execute=@n79.then_1, Value:float=6) -> (then, ReturnValue:Meta Sound Literal Structure)
   > Sequence()
   > SetGraphInputDefault(self=Get LowSpecEngineAudioBuilder(), InputName=REVVolume, Literal=@n83)
   Show Message Dialog(
    execute=@n8.else
    Title="To Text (String)"(InString=Error:)
    Message="To Text (String)"(InString="No rev models selected. Try selecting rev models, and then click \"Get Current Path\"")
    MessageType=Ok
    DefaultValue=No
    MessageCategory=Error
   )
   "On Selected (AudioButtonToggle)"!() -> (OutputDelegate, then, Button:Audio Button Toggle Object Reference) {is=title K2Node_ComponentBoundEvent}
   > n84 = GetSelectedAssetsOfClass(AssetClass=/Script/Engine.SoundBase)
   > Branch(Condition="Equal (Integer)"(A=Length(TargetArray:Array of Object References=@n84), B=1))
   > n85 = For Each Loop(Array=@n84)
   > n86 = Cast To MetaSoundSource(Object:Object Reference=@n85)
   > n87 = SpawnSoundAtLocation!(
    execute
    WorldContextObject:Object Reference
    Sound:Sound Base Object Reference=/VehicleAudio/Patches/MSS_EngineAudio_REV_ParamTester
    Location:Vector="0, 0, 0"
    Rotation:Rotator
    VolumeMultiplier:float=1
    PitchMultiplier:float=1
    StartTime:float=0
    AttenuationSettings:Sound Attenuation Object Reference
    ConcurrencySettings:Sound Concurrency Object Reference
    bAutoDestroy:bool=true
   ) -> (then, ReturnValue:Audio Component Object Reference)
   > n88 = Set EngineParamTesterSound(EngineParamTesterSound:Audio Component Object Reference=@n87) -> (Output_Get:Audio Component Object Reference)
   n89 = OnRPMChanged!() -> (OutputDelegate, then, OutputName:Name, Output:"Meta Sound Output Structure (by ref)") {is=title K2Node_CustomEvent}
   n90 = WatchOutput!(
    execute=@n88.then
    self:Meta Sound Output Subsystem Object Reference=Get MetaSoundOutputSubsystem!(WorldContext:Object Reference) -> (ReturnValue:Meta Sound Output Subsystem Object Reference) {is=title K2Node_GetSubsystem}
    AudioComponent:Audio Component Object Reference=@n87
    OutputName:Name=RPM
    OnOutputValueChanged:"Delegate (by ref)"=@n89
    AnalyzerName:Name
    AnalyzerOutputName:Name
   ) -> (then, ReturnValue:bool)
   n91 = OnThrottleChanged!() -> (OutputDelegate, then, OutputName:Name, Output:"Meta Sound Output Structure (by ref)") {is=title K2Node_CustomEvent}
   n92 = WatchOutput(execute=@n90.then, self=Get MetaSoundOutputSubsystem(), AudioComponent=@n87, OutputName=Throttle, OnOutputValueChanged=@n91)
   n93 = OnGearChanged!() -> (OutputDelegate, then, OutputName:Name, Output:"Meta Sound Output Structure (by ref)") {is=title K2Node_CustomEvent}
   WatchOutput(execute=@n92.then, self=Get MetaSoundOutputSubsystem(), AudioComponent=@n88, OutputName=Gear, OnOutputValueChanged=@n93)
   n94 = GetLevelViewportCameraInfo!(WorldContextObject:Object Reference) -> (CameraLocation:Vector, CameraRotation:Rotator, ReturnValue:bool)
   n95 = GetForwardVector!(InRot:Rotator=GetLevelViewportCameraInfo().CameraRotation) -> (ReturnValue:Vector)
   > n96 = Set Starting Camera Location(Starting Camera Location:Vector=vector + vector(A=@n94, B=vector * vector(A=@n95, B:float=2) -> (ReturnValue=~))) -> (Output_Get:Vector)
   > Set Camera Rotation(Camera Rotation:Rotator=@n94.CameraRotation) -> (Output_Get:Rotator)
   > Print Vector!(Execute:Exec, Vector:Vector=@n96, OptionalString:String, Duration:float=2, TextColor:Linear Color Structure=(R=0,G=0.6588,B=1,A=1)) -> (Then:Exec) {is=title K2Node_MacroInstance}
   > n97 = SpawnSoundAtLocation(Sound=@n86, Location=@n96, Rotation=@n94.CameraRotation, VolumeMultiplier=1, PitchMultiplier=1, StartTime=0, bAutoDestroy=true)
   > n98 = Set PreviewingSound(PreviewingSound:Audio Component Object Reference=@n97) -> (Output_Get:Audio Component Object Reference)
   > Play!(execute, self:Audio Component Object Reference=@n98, StartTime:float=0) -> (then)
   n99 = Get AnalogSlider_Throttle() -> (AnalogSlider_Throttle:Analog Slider Object Reference)
   > Set Boolean Parameter!(
    execute
    self:Audio Parameter Controller Interface Interface=Get EngineParamTesterSound() -> (EngineParamTesterSound:Audio Component Object Reference)
    InName:Name=Demo
    InBool:bool="float >= float"(A=Get Value!(self:Slider Object Reference=@n99) -> (Value:float), B=1)
   ) -> (then) {is=title K2Node_CallFunction}
   n100 = "On Deselected (AudioButtonToggle)"!() -> (OutputDelegate, then, Button:Audio Button Toggle Object Reference) {is=title K2Node_ComponentBoundEvent}
   n101 = Get PreviewingSound() -> (PreviewingSound:Audio Component Object Reference)
   FadeOut!(
    execute=[@n153.then, @n100.then]
    self:Audio Component Object Reference=@n101
    FadeOutDuration:float=0.1
    FadeVolumeLevel:float=0
    FadeCurve:EAudioFaderCurve Enum=Linear
   ) -> (then)
   > Destroy Component!(execute, self:Actor Component Object Reference=@n101, Object:Object Reference) -> (then)
   n102 = Get EngineParamTesterSound()
   > Stop(self:Audio Component Object Reference=@n102)
   > Destroy Component(self=@n102)
   > Set PreviewingSound()
   > Set Previewer Builder(Previewer Builder:Meta Sound Source Builder Object Reference) -> (Output_Get:Meta Sound Source Builder Object Reference)
   > Set EngineParamTesterSound()
   "On Pressed (EditorUtilityButton_56_GoToVehicleCosmetic)"() {is=title K2Node_ComponentBoundEvent}
   > n103 = Branch(Condition="Not Equal Exactly (String)"(A=Get SanitizedAssetName()))
   n104 = Append!(A=Get MutableBodiesPath() -> (MutableBodiesPath:String), B=/, C=Get SanitizedAssetName()) -> (ReturnValue)
   > n105 = DoesDirectoryExist!(execute, DirectoryPath:String=@n104) -> (then, ReturnValue:bool)
   > n106 = Branch(Condition=@n105)
   > SyncBrowserToFolders(FolderList=Make Array("[0]"=@n104))
   n107 = Get SanitizedAssetName()
   > n108 = FindAssetData!(execute, AssetPath:String=Append!(A=Get MutableBodiesPath(), B=/, C=@n107, D:String=/, E:String=VCID_Body_, F:String=@n107) -> (ReturnValue)) -> (then, ReturnValue:Asset Data Structure)
   > OpenEditorForAssets!(
    execute
    self:Asset Editor Subsystem Object Reference=Get AssetEditorSubsystem()
    Assets:Array of Object References=Make Array("[0]":Object Reference=GetAsset(InAssetData=@n108)) -> (Array:Array of Object References)
    OpenedMethod:EAssetTypeActivationOpenedMethod Enum=Edit
   ) -> (then, ReturnValue:bool)
   SyncBrowserToFolders(execute=[@n106.else, @n103.else], FolderList=Make Array("[0]":String=Get MutableBodiesPath()) -> (Array:Array of Strings))
   "On Clicked (EditorUtilityButton_56_ImportRevSettings)"() {is=title K2Node_ComponentBoundEvent}
   > n109 = Is Valid(InputObject=Get RevSettingsAsset())
   > n110 = Execute Python Script!(
    execute
    PythonScript="## Copyright Epic Games, Inc. All Rights Reserved.\r\n\r\nimport os\r\nimport sys\r\nimport unreal\r\nimport xml.etree.ElementTree as ElementTree\r\nimport tkinter as tk\r\nfrom   tkinter import filedialog\r\n\r\nwith unreal.ScopedEditorTransaction(\"Import REV Settings\"):\r\n    \r\n    ## Forward-declaration of values passed in from Blueprints.\r\n    rev_settings_asset: unreal.Object\r\n    initial_model_file_search_path: str\r\n    \r\n    ## Initialize tkinter, hide default window.\r\n    default_window = tk.Tk()\r\n    default_window.withdraw()\r\n    \r\n    ## Prompt user to pick .model file path.\r\n    selected_model_file_path = filedialog.askopenfilename(title=\"Import Settings from REV Model\", initialdir=initial_model_file_search_path, filetypes=[(\"REV Model File\", \"*.model\")])\r\n    if not os.path.exists(selected_model_file_path):\r\n        unreal.log_error(f\".model file not found: '{selected_model_file_path}'.\")\r\n        sys.exit(\"Invalid .model file path.\")\r\n    \r\n    ## Extract string contents of .model file.\r\n    model_file_contents: str\r\n    with open(file=selected_model_file_path, mode=\"rb\") as file:\r\n        model_file_contents = file.read().decode(errors=\"ignore\")\r\n    \r\n    ## The .model file contains some unreadable characters. \r\n    ## However, the '<EngineSimulationControlData>' section, which contains the REV settings, is still intact.\r\n    ## Sub-string the contents of the file to only include the '<EngineSimulationControlData>' section, before attempting to parse the XML data.\r\n    ## Otherwise, ElementTree.fromstring() will fail when it attempts to parse the garbage data contained in the rest of the file.\r\n    start_tag = \"<EngineSimulationControlData\"\r\n    end_tag = \"</EngineSimulationControlData>\"\r\n    control_data_substring_start_index = model_file_contents.find(start_tag)\r\n    control_data_substring_end_index = model_file_contents.find(end_tag) + len(end_tag)\r\n    if control_data_substring_start_index < 0 or control_data_substring_end_index < 0:\r\n        unreal.log_error(f\"'<EngineSimulationControlData>' schema not found in .model file: '{selected_model_file_path}'.\")\r\n        sys.exit()\r\n    \r\n    ## Parse sanitized XML string contents into an ElementTree object.\r\n    control_data_xml_tree = ElementTree.fromstring(text=model_file_contents[control_data_substring_start_index:control_data_substring_end_index])\r\n    \r\n    ## Search the ElementTree and attempt to set REVSettings data. Fail if any value is not found in the XML tree.\r\n    ## The values are returned as a string, and need to be cast to the appropriate property type.\r\n    ## ('true' is lowercase in the XML data, but the 'True' keyword in Python is capitalized, so text.capitalize() is used before casting to bool).\r\n    try:\r\n        rev_settings_asset.set_editor_property(name=\"clutch_rpm_merge_time\", value=float(control_data_xml_tree.find(\".//ClutchRPMMergeTime\").text), notify_mode=unreal.PropertyAccessChangeNotifyMode.ALWAYS)\r\n        rev_settings_asset.set_editor_property(name=\"clutch_rpm_spike\", value=float(control_data_xml_tree.find(\".//ClutchRPMSpike\").text), notify_mode=unreal.PropertyAccessChangeNotifyMode.ALWAYS)\r\n        rev_settings_asset.set_editor_property(name=\"clutch_rpm_spike_duration\", value=float(control_data_xml_tree.find(\".//ClutchRPMSpikeDuration\").text),  notify_mode=unreal.PropertyAccessChangeNotifyMode.ALWAYS)\r\n        rev_settings_asset.set_editor_property(name=\"down_shift_duration\", value=float(control_data_xml_tree.find(\".//DownShiftDuration\").text), notify_mode=unreal.PropertyAccessChangeNotifyMode.ALWAYS)\r\n        rev_settings_asset.set_editor_property(name=\"pop_duration\", value=float(control_data_xml_tree.find(\".//PopDuration\").text), notify_mode=unreal.PropertyAccessChangeNotifyMode.ALWAYS)\r\n        rev_settings_asset.set_editor_property(name=\"up_shift_attack_duration\", value=float(control_data_xml_tree.find(\".//UpShiftAttackDuration\").text), notify_mode=unreal.PropertyAccessChangeNotifyMode.ALWAYS)\r\n        rev_settings_asset.set_editor_property(name=\"up_shift_attack_rpm\", value=float(control_data_xml_tree.find(\".//UpShiftAttackRPM\").text), notify_mode=unreal.PropertyAccessChangeNotifyMode.ALWAYS)\r\n        rev_settings_asset.set_editor_property(name=\"up_shift_attack_throttle_time\", value=float(control_data_xml_tree.find(\".//UpShiftAttackThrottleTime\").text), notify_mode=unreal.PropertyAccessChangeNotifyMode.ALWAYS)\r\n        rev_settings_asset.set_editor_property(name=\"up_shift_attack_volume_spike\", value=float(control_data_xml_tree.find(\".//UpShiftAttackVolumeSpike\").text), notify_mode=unreal.PropertyAccessChangeNotifyMode.ALWAYS)\r\n        rev_settings_asset.set_editor_property(name=\"up_shift_duration\", value=float(control_data_xml_tree.find(\".//UpShiftDuration\").text),  notify_mode=unreal.PropertyAccessChangeNotifyMode.ALWAYS)\r\n        rev_settings_asset.set_editor_property(name=\"up_shift_wobble_duration\", value=float(control_data_xml_tree.find(\".//UpShiftWobbleDuration\").text), notify_mode=unreal.PropertyAccessChangeNotifyMode.ALWAYS)\r\n        rev_settings_asset.set_editor_property(name=\"up_shift_wobble_enabled\", value=bool(control_data_xml_tree.find(\".//UpShiftWobbleEnabled\").text.capitalize()), notify_mode=unreal.PropertyAccessChangeNotifyMode.ALWAYS)\r\n        rev_settings_asset.set_editor_property(name=\"up_shift_wobble_pitch_amp\", value=float(control_data_xml_tree.find(\".//UpShiftWobblePitchAmp\").text), notify_mode=unreal.PropertyAccessChangeNotifyMode.ALWAYS)\r\n        rev_settings_asset.set_editor_property(name=\"up_shift_wobble_pitch_freq\", value=float(control_data_xml_tree.find(\".//UpShiftWobblePitchFreq\").text), notify_mode=unreal.PropertyAccessChangeNotifyMode.ALWAYS)\r\n        rev_settings_asset.set_editor_property(name=\"up_shift_wobble_vol_amp\", value=float(control_data_xml_tree.find(\".//UpShiftWobbleVolAmp\").text), notify_mode=unreal.PropertyAccessChangeNotifyMode.ALWAYS)\r\n        rev_settings_asset.set_editor_property(name=\"up_shift_wobble_vol_freq\", value=float(control_data_xml_tree.find(\".//UpShiftWobbleVolFreq\").text), notify_mode=unreal.PropertyAccessChangeNotifyMode.ALWAYS)\r\n        unreal.log(f\"Successfully imported REV settings from .model file '{selected_model_file_path}'\")\r\n    except Exception as e:\r\n        unreal.log_error(f\"Could not update RevSettingsAsset properties from XML tree: {e}\")"
    rev_settings_asset:Rev Settings Object Reference=Get RevSettingsAsset()
    initial_model_file_search_path:String=Get Current Model File Path!() -> (ReturnValue:String) {is=title K2Node_MacroInstance}
   ) -> (then, ReturnValue, selected_model_file_path:String) {inputs [rev_settings_asset, initial_model_file_search_path], outputs [selected_model_file_path]}
   > n111 = Branch(Condition=@n110)
   > Set ModelFilePath(ModelFilePath:File Path Structure=Make File Path!(FilePath:String=@n110.selected_model_file_path) -> (FilePath:File Path Structure) {is=title K2Node_MakeStruct}) -> (Output_Get:File Path Structure)
   > PlayForward(self:UIAnimation Controller Object Reference=Get RevSettingsSuccessAnimation() -> (RevSettingsSuccessAnimation:Timeline Animation Controller Object Reference))
   PlayForward(execute=@n111.else, self=Get RevSettingsFailureAnimation() -> (RevSettingsFailureAnimation:Timeline Animation Controller Object Reference))
   Show Message Dialog(
    execute=@n109.Is Not Valid
    Title=Format Text(Format="LOCTEXT(\"Error - REV Settings Import\")")
    Message=Format Text(Format="LOCTEXT(\"REVSettings asset is unset.\")")
    MessageType=Ok
    DefaultValue=Ok
    MessageCategory=Error
   )
   > PlayForward(self=Get RevSettingsFailureAnimation())
   "On Pressed (EditorUtilityButton_Authoring)"() {is=title K2Node_ComponentBoundEvent}
   n112 = Get CommonActivatableWidgetSwitcher_171() -> (CommonActivatableWidgetSwitcher_171:Common Activatable Widget Switcher Object Reference)
   > Set ActiveWidgetIndex!(execute, ActiveWidgetIndex:int=0, self:Widget Switcher Object Reference=@n112) -> (then, Output_Get:int=0)
   n113 = "On Value Changed (RadialSlider_Azimuth)"!() -> (OutputDelegate, then, Value:float) {is=title K2Node_ComponentBoundEvent}
   > Set World Rotation!(
    execute
    self:Scene Component Object Reference=Get PreviewingSound()
    NewRotation:Rotator=MakeRotFromX!(X:"Vector (by ref)"="To Vector (Float)"!(InDouble:double=@n113.Value) -> (ReturnValue:Vector)) -> (ReturnValue:Rotator)
    bSweep:bool=false
    bTeleport:bool=false
   ) -> (then, SweepHitResult:Hit Result Structure) {is=title K2Node_CallFunction}
   n114 = "On Selected (AudioButtonToggle_FNVehicleMix)"!() -> (OutputDelegate, then, Button:Audio Button Toggle Object Reference) {is=title K2Node_ComponentBoundEvent}
   n115 = Get AudioButtonToggle_FNVehicleMix() -> (AudioButtonToggle_FNVehicleMix:Audio Button Toggle Object Reference)
   n116 = GetIsSelected!(execute=[@n114.then, @n119.then], self:Audio Button Toggle Object Reference=@n115) -> (then, IsSelected:bool)
   > n117 = Branch(Condition=@n116)
   n118 = Get Mix() -> (Mix:Sound Control Bus Mix Object Reference)
   Activate Control Bus Mix!(execute=[@n154.then, @n117.then], WorldContextObject:Object Reference, Mix:Sound Control Bus Mix Object Reference=@n118) -> (then)
   Deactivate Control Bus Mix!(execute=@n117.else, WorldContextObject:Object Reference, Mix:Sound Control Bus Mix Object Reference=@n118) -> (then)
   n119 = "On Deselected (AudioButtonToggle_FNVehicleMix)"!() -> (OutputDelegate, then, Button:Audio Button Toggle Object Reference) {is=title K2Node_ComponentBoundEvent}
   "On Pressed (EditorUtilityButton_56_Importfiles)"() {is=title K2Node_ComponentBoundEvent}
   > n120 = GetAssetTools()
   n121 = "To Text (String)"(InString=GetCurrentContentBrowserPath())
   n122 = Get FolderDirectory() -> (FolderDirectory:String)
   > n123 = ImportAssetsWithDialog!(
    execute
    self:Asset Tools Interface=@n120
    DestinationPath:String=SelectString(A="To String (Text)"(InText=@n121), B=@n122, bPickA=IsEmpty(InString=@n122))
   ) -> (then, ReturnValue:Array of Object References)
   > SaveLoadedAssets(AssetsToSave=@n123, bOnlyIfIsDirty=false)
   n124 = "On Value Changed (RadialSlider_Distance)"!() -> (OutputDelegate, then, Value:float) {is=title K2Node_ComponentBoundEvent}
   n125 = GetForwardVector(InRot=Get Camera Rotation() -> (Camera Rotation:Rotator))
   n126 = Get PreviewingSound()
   > Set World Location!(
    execute
    self:Scene Component Object Reference=@n126
    NewLocation:Vector=vector + vector(A=Get Starting Camera Location() -> (Starting Camera Location:Vector), B=vector * vector(A=@n125, B=@n124.Value) -> (ReturnValue=~))
    bSweep:bool=false
    bTeleport:bool=false
   ) -> (then, SweepHitResult:Hit Result Structure) {is=title K2Node_CallFunction}
   > Print Vector(
    Vector=Get World Location!(self:Scene Component Object Reference=@n126) -> (ReturnValue:Vector) {is=title K2Node_CallFunction}
    Duration=2
    TextColor=(R=0,G=0.6588,B=1,A=1)
   )
   > Set Text(
    Text="To Text (String)"(InString="To String (Float)"(InDouble=@n124.Value))
    self=Get TextBlock_ExampleAssetName_1() -> (TextBlock_ExampleAssetName_1:Text Object Reference)
   )
   "On Pressed (EditorUtilityButton_56_MakeDataAssets)"() {is=title K2Node_ComponentBoundEvent}
   n127 = Get Text(self=Get AssetNameInput())
   > n128 = Branch(Condition="Equal (Boolean)"(A=TextIsEmpty(InText=@n127)) -> (ReturnValue=~))
   > n129 = Get!(execute:Exec) -> (then:Exec, else:Exec, Engine Sound:Sound Base Soft Object Reference)
   n130 = GetAssetTools(execute=[@n144.Yes, @n144.YesAll, @n129.then])
   > Set AssetTools(AssetTools:Object Reference="To Object (Interface)"(Interface=@n130)) -> (Output_Get:Object Reference)
   n131 = Append!(A=Veh_Item_EngineAudio_, B=Get SanitizedAssetName()) -> (ReturnValue)
   > n132 = CheckIfAssetExists(AssetName=@n131, AssetPath=Get Data Asset Package Path() -> (Data Asset Package Path:String))
   > n133 = Branch(Condition=@n132)
   n134 = CreateAsset(
    execute=@n133.else
    self=Get AssetTools() -> (AssetTools:Object Reference)
    AssetName=@n131
    PackagePath=Get Data Asset Package Path()
    AssetClass=/Script/VehicleCosmeticsRuntime.VehicleCosmeticsItemDefinition_EngineAudio
    CallingContext=None
    bOverwriteExisting=false
   )
   > Set VehicleEngineAudioItemDefinition(VehicleEngineAudioItemDefinition:Object Reference=@n134) -> (Output_Get:Object Reference)
   > SaveDirectory(DirectoryPath=Get Data Asset Package Path(), bOnlyIfIsDirty=true, bRecursive=false)
   > n135 = Sequence()
   > Branch()
   > n136 = MakeInstancedStruct!(
    execute:Exec
    Value:Motor Physics Sim Config Data Structure=Make Motor Physics Sim Config Data!(
     Weight:float=Get Weight() -> (Weight:float)
     EngineTorque:float=Get Torque() -> (Torque:float)
     BrakingHorsePower:float=Get Braking Horsepower() -> (Braking Horsepower:float)
    ) -> (MotorPhysicsSimConfigData:Motor Physics Sim Config Data Structure) {is=title K2Node_MakeStruct}
   ) -> (then:Exec, ReturnValue:Instanced Struct Structure) {is=title K2Node_InstancedStruct}
   > n137 = MakeInstancedStruct(Value:Vehicle Audio Cosmetic Data Struct Structure=Make Vehicle Audio Cosmetic Data Struct!(
     EngineSound:Sound Base Soft Object Reference=Get Engine Sound() -> (Engine Sound:Sound Base Soft Object Reference)
     EngineIgnitionSound:Sound Base Object Reference=Get Engine Ignition Sound() -> (Engine Ignition Sound:Sound Base Object Reference)
     EngineOffSound:Sound Base Object Reference=Get Engine Off Sound() -> (Engine Off Sound:Sound Base Object Reference)
     GearShiftUpSound:Sound Base Object Reference
     GearShiftDownSound:Sound Base Object Reference
     HornSound:Sound Base Object Reference=Get Horn Sound() -> (Horn Sound:Sound Base Object Reference)
     bIsIP:bool=Get Is IP() -> (Is IP:bool)
     bIgnoreShifting:bool=Get Ignore Shifting() -> (Ignore Shifting:bool)
     bIgnoreExhaustPops:bool=Get Ignore Exhaust Pops() -> (Ignore Exhaust Pops:bool)
     MotorSimConfigData:Array of Instanced Struct Structures=Make Array("[0]":Instanced Struct Structure=@n136) -> (Array:Array of Instanced Struct Structures)
    ) -> (VehicleAudioCosmeticDataStruct:Vehicle Audio Cosmetic Data Struct Structure) {is=title K2Node_MakeStruct})
   n138 = Append(A=CT_VehicleCosmeticsAudio_, B=Get SanitizedAssetName())
   > n139 = CheckIfAssetExists(AssetName=@n138, AssetPath=Get Data Asset Package Path())
   > n140 = Branch(Condition=@n139)
   n141 = CreateVehicleAudioCosmeticChooser!(
    execute=@n140.else
    AssetName:String=@n138
    AssetPath:String=Get Data Asset Package Path()
    FallbackCosmeticData:"Instanced Struct Structure (by ref)"=@n137
    bAddToSourceControl:bool=true
   ) -> (then, ReturnValue:Chooser Table Object Reference)
   > Set CosmeticChooserTableSoftPtr(CosmeticChooserTableSoftPtr:Chooser Table Soft Object Reference=Make Soft Reference(Input:Chooser Table Object Reference=@n141) -> (Output:Chooser Table Soft Object Reference)) -> (Output_Get:Chooser Table Soft Object Reference)
   > SaveLoadedAsset(AssetToSave=@n141, bOnlyIfIsDirty=false)
   > SetEditorProperty(
    Object=Get VehicleEngineAudioItemDefinition() -> (VehicleEngineAudioItemDefinition:Object Reference)
    PropertyName=DataChooserTable
    PropertyValue:Chooser Table Soft Object Reference=Get CosmeticChooserTableSoftPtr() -> (CosmeticChooserTableSoftPtr:Chooser Table Soft Object Reference)
   )
   n142 = Get Data Asset Package Path()
   SaveDirectory(execute=@n135.then_1, DirectoryPath=@n142, bOnlyIfIsDirty=true, bRecursive=false)
   > SaveDirectory(DirectoryPath=@n142, bOnlyIfIsDirty=true, bRecursive=false)
   n143 = Show Message Dialog(
    execute=@n129.else
    Title="To Text (String)"(InString=Warning)
    Message="To Text (String)"(InString="No Engine Audio is defined in the Engine Audio Creator, so no Engine audio will be populated. Are you sure you want to create Data Assets? Clicking Yes will continue the process. ")
    MessageType=YesNo
    DefaultValue=No
    MessageCategory=Warning
   )
   > n144 = Switch on EAppReturnType(Selection=@n143)
   Show Message Dialog(
    execute=@n128.else
    Title="To Text (String)"(InString=Error:)
    Message="To Text (String)"(InString="No asset name input selected. Please input an asset name. ")
    MessageType=Ok
    DefaultValue=No
    MessageCategory=Error
   )
   "On Pressed (EditorUtilityButton_56_GoToVehicleAudioDataAssets)"() {is=title K2Node_ComponentBoundEvent}
   > SyncBrowserToFolders(FolderList=Make Array("[0]":String=/VehicleCosmetics/EngineAudio) -> (Array:Array of Strings))
   "On Pressed (EditorUtilityButton_56_GoToVehicleAudio)"() {is=title K2Node_ComponentBoundEvent}
   > SyncBrowserToFolders(FolderList=Make Array("[0]"=/VehicleAudio/EngineAudio))
   "On Pressed (EditorUtilityButton_Loudness)"() {is=title K2Node_ComponentBoundEvent}
   > Set ActiveWidgetIndex(ActiveWidgetIndex=2, self=@n112) -> (Output_Get=0)
   n145 = "On Value Changed (AnalogSlider_Throttle)"!() -> (OutputDelegate, then, Value:float) {is=title K2Node_ComponentBoundEvent}
   > Set Boolean Parameter(self=Get EngineParamTesterSound(), InName=Demo, InBool="float >= float"(A=@n145.Value, B=1))
   "On Pressed (EditorUtilityButton_PreviewButton)"() {is=title K2Node_ComponentBoundEvent}
   > Set ActiveWidgetIndex(ActiveWidgetIndex=1, self=@n112) -> (Output_Get=0)
   Event Tick()
   n146 = Get PreviewingSound()
   > n147 = Branch(Condition=IsPlaying!(self:Audio Component Object Reference=@n146) -> (ReturnValue:bool))
   n148 = Get World Location(self=@n146)
   n149 = Get World Rotation!(self:Scene Component Object Reference=Get PreviewingSound()) -> (ReturnValue:Rotator) {is=title K2Node_CallFunction}
   > DrawDebugArrow!(
    execute
    WorldContextObject:Object Reference
    LineStart:Vector=@n148
    LineEnd:Vector=vector + vector(A=@n148, B=GetForwardVector(InRot=@n149))
    ArrowSize:float=50
    LineColor:Linear Color Structure=(R=1,G=0,B=0.159,A=1)
    Duration:float=0
    Thickness:float=1
    DepthPriority:EDrawDebugSceneDepthPriorityGroup Enum=World
   ) -> (then)
   > n150 = DrawDebugSphere!(
    execute
    WorldContextObject:Object Reference
    Center:Vector=@n148
    Radius:float=100
    Segments:int=12
    LineColor:Linear Color Structure=(R=0.0549,G=0.5417,B=0.1766,A=1)
    Duration:float=0
    Thickness:float=0
    DepthPriority:EDrawDebugSceneDepthPriorityGroup Enum=World
   ) -> (then)
   n151 = Branch(execute=[@n150.then, @n147.else], Condition=IsPlaying(self=Get PreviewingSound()))
   n152 = "To String (Text)"(InText=Get Text(self=Get AssetNameInput()))
   Set SanitizedAssetName(
    execute=[@n151.else, @n151.then]
    SanitizedAssetName:String=TrimTrailing!(SourceString:String=Trim!(SourceString:String=@n152) -> (ReturnValue:String)) -> (ReturnValue:String)
   ) -> (Output_Get:String)
   > Branch(Condition=IsEmpty(InString=Get SanitizedAssetName()))
   > Set Text(
    Text="To Text (String)"(InString=GetCurrentContentBrowserPath())
    self:Text Box Object Reference=Get PathInput() -> (PathInput:Editor Utility Editable Text Box Object Reference)
   )
   Event Destruct()
   > n153 = Deactivate Control Bus Mix(Mix=Get Mix())
   Event PreConstruct()
   > SetObject(
    self=[
     Get SinglePropertyView_Ignition() -> (SinglePropertyView_Ignition:Single Property View Object Reference)
     Get SinglePropertyView_IgnoreExhaustPops() -> (SinglePropertyView_IgnoreExhaustPops:Single Property View Object Reference)
     Get SinglePropertyView_IsIP() -> (SinglePropertyView_IsIP:Single Property View Object Reference)
     Get SinglePropertyView_Off() -> (SinglePropertyView_Off:Single Property View Object Reference)
     Get SinglePropertyView_EngineSound_Weight() -> (SinglePropertyView_EngineSound_Weight:Single Property View Object Reference)
     Get SinglePropertyView_EngineSound_Torque() -> (SinglePropertyView_EngineSound_Torque:Single Property View Object Reference)
     Get SinglePropertyView_EngineSound_BrakingHorsepower() -> (SinglePropertyView_EngineSound_BrakingHorsepower:Single Property View Object Reference)
     Get SinglePropertyView_Horn() -> (SinglePropertyView_Horn:Single Property View Object Reference)
     Get SinglePropertyView_EngineSound_VehicleItemDataAsset() -> (SinglePropertyView_EngineSound_VehicleItemDataAsset:Single Property View Object Reference)
     Get SinglePropertyView_EngineSound_ChooserTable() -> (SinglePropertyView_EngineSound_ChooserTable:Single Property View Object Reference)
     Get SinglePropertyView_EngineSound() -> (SinglePropertyView_EngineSound:Single Property View Object Reference)
     Get SinglePropertyView_EngineSound_RevSettings() -> (SinglePropertyView_EngineSound_RevSettings:Single Property View Object Reference)
     Get SinglePropertyView_IgnoreShifting() -> (SinglePropertyView_IgnoreShifting:Single Property View Object Reference)
    ]
    NewObject=Self-Reference()
   )
   > Set ColorAndOpacity!(
    execute
    ColorAndOpacity:Linear Color Structure=(R=0,G=0,B=0,A=0)
    self:Image Object Reference=Get REVSettingsBackground_Green() -> (REVSettingsBackground_Green:Image Object Reference)
   ) -> (then, Output_Get:Linear Color Structure)
   > Set ColorAndOpacity(ColorAndOpacity=(R=0,G=0,B=0,A=0), self=Get REVSettingsBackground_Red() -> (REVSettingsBackground_Red:Image Object Reference))
   Event Construct()
   > InitializeMuteAndSolo()
   > Create Timeline Animation Controllers!(execute:Exec) -> (then:Exec) {is=title K2Node_MacroInstance}
   > n154 = SetSelected!(execute, self:Audio Button Toggle Object Reference=@n115, ReportEvent:bool=true) -> (then)
   Branch(Condition=Is Valid Index!(TargetArray:Array of Wildcards, IndexToTest:int) -> (ReturnValue:bool))
   > n155 = Create Widget!(
    execute:Exec
    WorldContextObject:Object Reference
    Class:User Widget Class Reference=&/EUW_ExhaustPopButton.EUW_ExhaustPopButton_C
    OwningPlayer:Player Controller Object Reference
   ) -> (then:Exec, ReturnValue:EUW Exhaust Pop Button Object Reference)
   > AddToViewport!(execute, self:User Widget Object Reference=@n155, ZOrder:int=0) -> (then)
   > SetPositionInViewport!(execute, self:User Widget Object Reference=@n155, Position:Vector 2D Structure, bRemoveDPIScale:bool=true) -> (then)
   Get Rotation X Vector!(InRot:Rotator=@n149) -> (ReturnValue:Vector) {is=title K2Node_CallFunction}
   n156 = Get VehicleCosmeticDataSoftObject() -> (VehicleCosmeticDataSoftObject:Soft Object Reference)
   n157 = Get CosmeticChooserTableSoftPtr()
   OpenEditorForAssets(
    self=Get AssetEditorSubsystem()
    Assets=Make Array!(
     "[0]":Object Reference=Get VehicleEngineAudioItemDefinition()
     "[1]":Object Reference=Resolve Soft Reference!(Input:Chooser Table Soft Object Reference=@n157) -> (Output:Chooser Table Object Reference)
     "[2]":Object Reference=Resolve Soft Reference(Input:Soft Object Reference=@n156) -> (Output:Object Reference)
    ) -> (Array:Array of Object References)
    OpenedMethod=Edit
   )
   n158 = GetFloat!(execute=@n89.then, Output:"Meta Sound Output Structure (by ref)"=@n89.Output) -> (then, Success:bool, ReturnValue:float)
   n159 = Get RadialSlider_Tachometer() -> (RadialSlider_Tachometer:Radial Slider Object Reference)
   SetSliderProgressColor!(
    execute=@n164.then
    self:Radial Slider Object Reference=@n159
    InValue:Linear Color Structure="Lerp (LinearColor)"(B=(R=1,G=0,B=0,A=1), Alpha=@n158.ReturnValue)
   ) -> (then)
   n160 = GetFloat(execute=@n91.then, Output=@n91.Output)
   SetCenterBackgroundColor!(
    execute=@n163.then
    self:Radial Slider Object Reference=@n159
    InValue:Linear Color Structure="Lerp (LinearColor)"(B=(R=0,G=1,B=0.2748,A=1), Alpha=@n160.ReturnValue)
   ) -> (then)
   Get Camera Rotation()
   Get SinglePropertyView_EngineSound()
   n161 = GetInt32!(execute=@n93.then, Output:"Meta Sound Output Structure (by ref)"=@n93.Output) -> (then, Success:bool, ReturnValue:int=0)
   Set Text(
    execute=@n165.then
    Text="To Text (Integer)"(Value=@n161.ReturnValue)
    self:Text Object Reference=Get Text_GearNumber() -> (Text_GearNumber:Text Object Reference)
   )
   n162 = Get PreviewingSound()
   n163 = Set Float Parameter!(
    execute=@n160.then
    self:Audio Parameter Controller Interface Interface=@n162
    InName:Name=Fort.Vehicle.Throttle
    InFloat:float=@n160.ReturnValue
   ) -> (then) {is=title K2Node_CallFunction}
   Set Float Parameter(execute=@n158.then, self=Get PreviewingSound(), InName=Fort.Vehicle.RPM, InFloat=@n158.ReturnValue)
   > n164 = Set Value!(execute, Value:float=@n158.ReturnValue, self:Radial Slider Object Reference=@n159) -> (then, Output_Get:float)
   n165 = Set Integer Parameter!(execute=@n161.then, self:Audio Parameter Controller Interface Interface=@n162, InName:Name=Fort.Vehicle.Gear, InInt:int=@n161.ReturnValue) -> (then) {is=title K2Node_CallFunction}
  ]}
 ChannelCountWarning {
  path=EventGraph.K2Node_Composite_0.ChannelCountWarning
  nodes [
   Inputs()
   > Show Message Dialog(
    Title="To Text (String)"(InString=Error:)
    Message="To Text (String)"(InString="Your REV models have a non-standard channel count, double check your REV files are the correct amount of channels (Mono)")
    MessageType=Ok
    DefaultValue=No
    MessageCategory=Error
   )
   Outputs!()
  ]
 }
 SampleRateError {
  path=EventGraph.K2Node_Composite_1.SampleRateError
  nodes [
   Inputs()
   n1 = Append!(
    A="To String (Float)"(InDouble=Get RevFullResolutionSampleRate())
    B="and "
    C:String="To String (Float)"(InDouble=Get RevMobileResolutionSampleRate())
   ) -> (ReturnValue)
   > Show Message Dialog(
    Title="To Text (String)"(InString=Error:)
    Message="To Text (String)"(InString=Append!(
      A="Your REV models have a non-standard sample rate, double check you are using the correct sample rate for the REV .mod files. Standard rates are:"
      B=@n1
     ) -> (ReturnValue))
    MessageType=Ok
    DefaultValue=No
    MessageCategory=Error
   )
   Outputs()
  ]
 }
 GenerateMetasoundBuilder {nodes [
   n1 = GenerateMetasoundBuilder!() -> (then, Object:Object Reference) {is=title K2Node_FunctionEntry}
   > n2 = Cast To MetaSoundDocumentInterface!(execute, Object=@n1) -> (then, CastFailed, AsMetaSound Document Interface:Meta Sound Document Interface Interface) {is=title K2Node_DynamicCast}
   n3 = "To String (Text)"(InText="To Text (Object)"!(InObj:Object Reference=@n1) -> (ReturnValue:Text))
   > n4 = CreateSourcePresetBuilder!(
    execute
    self:Meta Sound Builder Subsystem Object Reference=Get MetaSoundBuilderSubsystem()
    BuilderName:Name=String To Name(InString=Append(A=Builder_, B=@n3))
    ReferencedSourceClass:"Meta Sound Document Interface Interface (by ref)"=@n2
   ) -> (Succeeded:Exec, Failed:Exec, ReturnValue:Meta Sound Source Builder Object Reference)
   Return Node!(execute=@n4.Failed, Source Preset Builder:Meta Sound Source Builder Object Reference)
   > Return Node(Source Preset Builder=@n4)
  ]}
 GetCurrentModelFilePath {nodes [
   Inputs!()
   n1 = Get!() -> (ModelFilePath_FilePath:String)
   Outputs!(ReturnValue:String=Select(Option 0:String=Get(), Option 1:String=GetProjectDirectory!() -> (ReturnValue:String), Index=IsEmpty(InString=@n1)) -> (ReturnValue:String))
  ]}
 MuteAndSoloEventGraph {nodes [
   "On Pressed (EditorUtilityButton_HelpButton)"() {is=title K2Node_ComponentBoundEvent}
   > LaunchURL!(execute, URL:String=Get EngineAudioURL() -> (EngineAudioURL:String)) -> (then)
   InitializeMuteAndSolo() {is=title K2Node_CustomEvent}
   > n1 = Set AllSoloButtons(AllSoloButtons:Array of WBP Mute Solo Button Object References=Make Array!(
     "[0]":WBP Mute Solo Button Object Reference=Get WBP_MuteSoloButton_Ignition() -> (WBP_MuteSoloButton_Ignition:WBP Mute Solo Button Object Reference)
     "[1]":WBP Mute Solo Button Object Reference=Get WBP_MuteSoloButton_Off() -> (WBP_MuteSoloButton_Off:WBP Mute Solo Button Object Reference)
     "[2]":WBP Mute Solo Button Object Reference=Get WBP_MuteSoloButton_Horn() -> (WBP_MuteSoloButton_Horn:WBP Mute Solo Button Object Reference)
     "[3]":WBP Mute Solo Button Object Reference=Get WBP_MuteSoloButton_Engine() -> (WBP_MuteSoloButton_Engine:WBP Mute Solo Button Object Reference)
    ) -> (Array:Array of WBP Mute Solo Button Object References)) -> (Output_Get:Array of WBP Mute Solo Button Object References)
   > n2 = For Each Loop(Array:Array of WBP Mute Solo Button Object References=@n1) -> (Array Element:WBP Mute Solo Button Object Reference)
   n3 = OnCheckboxChanged!() -> (OutputDelegate, then, bIsChecked:bool=~, SendingWidget:Object Reference) {is=title K2Node_CustomEvent}
   Bind Event to On Checkbox Changed(execute=@n2.LoopBody, self:WBP Mute Solo Button Object Reference=@n2, Delegate=@n3) {is=title K2Node_AddDelegate}
   n4 = For Each Loop(Exec=@n3.then, Array=Get AllSoloButtons() -> (AllSoloButtons:Array of WBP Mute Solo Button Object References))
   Set CheckedState!(
    execute=@n5.then
    CheckedState:ECheckBoxState Enum=Unchecked
    self:Check Box Object Reference=Get CheckBox_26!(self:WBP Mute Solo Button Object Reference=@n4) -> (CheckBox_26:Check Box Object Reference)
   ) -> (then, Output_Get:ECheckBoxState Enum=Unchecked)
   n5 = Branch(
    execute=@n4.LoopBody
    Condition="Not Equal (Object)"(A:Object Reference=@n4, B:Object Reference=@n3.SendingWidget) -> (ReturnValue:bool=~) {is=title K2Node_PromotableOperator}
   )
  ]}
]
variables [
 3CHREVSizeLimit {type=integer, default=4000}
 AllSoloButtons {type="array<WBP_MuteSoloButton_C>", default []}
 Array Element {type="array<Object>", default []}
 As Meta Sound Source {type=MetaSoundSource, default=/Game/Sounds/Fort_Music/Music_MP_FrontEnd/wav/MS_MultiProduct_Amb_Bed, class=~, properties=~}
 AssetTools {type=Object, default=null}
 Braking Horsepower {type=number, default=6000}
 bREVOverBudget {type=boolean}
 Camera Rotation {type=Rotator}
 CosmeticChooserTableSoftPtr {type=ChooserTable, default=null}
 Data Asset Package Path {type=string, default=/VehicleCosmetics/EngineAudio}
 Engine Ignition Sound {type=SoundBase, default=null}
 Engine Off Sound {type=SoundBase, default=null}
 Engine Sound {type=SoundBase, default=null}
 EngineAudioURL {type=string, default="https://epicgames.atlassian.net/wiki/spaces/FORT/pages/75908916/Creating+Engine+Audio+Content##Implementations"}
 EngineParamTesterSound {type=AudioComponent, default=null}
 FolderDirectory {type=string, default=""}
 Forward Direction {type=Vector}
 GeneratedMetasounds {type="array<object>", default []}
 HighSpecEngineAudioBuilder {type=MetaSoundSourceBuilder, default=null}
 HighSpecEngineAudioMSS {type=Object, default=/VehicleAudio/Patches/MSS_EngineAudio_Parent_REV}
 Horn Preset MSS {type=Object, default=&MSS_Vehicle_Horn}
 Horn Preset MSS Builder {type=MetaSoundSourceBuilder, default=null}
 Horn Sound {type=SoundBase, default=null}
 Ignition Preset MSS {type=Object, default=&MSS_EngineAudio_Ignition}
 Ignition Preset MSS Builder {type=MetaSoundSourceBuilder, default=null}
 IgnitionWaveArrayInput {type="array<Object>", default []}
 Ignore Exhaust Pops {type=boolean}
 Ignore Shifting {type=boolean}
 Is IP {type=boolean}
 LowSpecEngineAudioBuilder {type=MetaSoundSourceBuilder, default=null}
 LowSpecEngineAudioMSS {type=Object, default=&MSS_EngineAudio_LowSpec_Parent_REV}
 Mix {type=SoundControlBusMix, default=null}
 ModelFilePath {type=FilePath, default {filePath=""}}
 MonoREVModelName {type=string, default=""}
 MonoREVSize {type=integer}
 MonoREVSizeLimit {type=integer, default=1800}
 MutableBodiesPath {type=string, default=/VehicleCosmetics/Mutable/Bodies}
 NewVar {type=SoundBase, default=null}
 NewVar_0 {type=ChooserTable, default=/Game/Animation/Game/MainPlayerFN/ChooserTables/AgileChugJug/AgileChugJug_Layering_CT}
 Node Class {default=MetaSoundSource'&MSS_EngineAudio_REV_ParamTester'}
 Off Preset MSS {type=Object, default=&MSS_EngineAudio_Off}
 Off Preset MSS Builder {type=MetaSoundSourceBuilder, default=null}
 OffWaveArrayInput {type="array<Object>", default []}
 Previewer Builder {type=MetaSoundSourceBuilder, default=null}
 PreviewingSound {type=AudioComponent, default=null}
 RevFullResolutionSampleRate {type=number, default=48000}
 RevMobileResolutionSampleRate {type=number, default=24000}
 RevSettingsAsset {type=RevSettings, default=null}
 RevSettingsFailureAnimation {type=UITimelineAnimationController, default=null}
 RevSettingsSuccessAnimation {type=UITimelineAnimationController, default=null}
 SanitizedAssetName {type=string, default=""}
 Starting Camera Location {type=Vector}
 Target {type=Object, default=null}
 Torque {type=number, default=2500}
 VehicleCosmeticDataSoftObject {type=Object, default=null}
 VehicleEngineAudioItemDefinition {type=Object, default=null}
 Weight {type=number, default=900}
 WorkunitPath {type=string, default="This\\Is\\My\\Path"}
]
widgets {
 info {rootWidgetClass=&WrapBox, widgetCount=90}
 widgets [
  SizeBox {widgetClassPath=&WrapBox, bIsVariable=true, widget=SizeBox, properties {bOverride_Cursor=true}}
  UIKitGridPanel_616 {widgetClassPath=/Script/FortniteUI.UIKitGridPanel, parent=SizeBox, widget=UIKitGridPanel_616, slot {class=WrapBoxSlot}}
  GridPanelAssetPathName {widgetClassPath=&GridPanel, parent=UIKitGridPanel_616, widget=GridPanelAssetPathName, slot {class=GridSlot}}
  UEFN_TextBlock_C_2 {
   widgetClassPath=&UEFN_TextBlock
   parent=GridPanelAssetPathName
   widget=UEFN_TextBlock_C_2
   properties {
    text=Asset Name:
    font {
     bForceMonospaced=false
     bMaterialIsStencil=false
     fontMaterial=null
     fontObject=&Roboto
     letterSpacing=0
     monospacedWidth=1
     size=19.5
     skewAmount=0
     typefaceFontName=Regular
     outlineSettings {
      bApplyOutlineToDropShadows=false
      bMiteredCorners=false
      bSeparateFillAlpha=false
      outlineMaterial=null
      outlineSize=0
      outlineColor=<rgba 0 0 0 1>
     }
    }
   }
   slot {class=GridSlot, properties {horizontalAlignment=HAlign_Right, row=1, verticalAlignment=VAlign_Center}}
  }
  PathInput {
   widgetClassPath=&EditorUtilityEditableTextBox
   bIsVariable=true
   parent=GridPanelAssetPathName
   widget=PathInput
   properties {widgetStyle {
     backgroundColor {colorUseRule=UseColor_Specified, specifiedColor=<rgba 1 1 1 1>}
     backgroundImageFocused {
      bIsDynamicallyLoaded=false
      drawAs=RoundedBox
      imageType=NoImage
      mirroring=NoMirror
      resourceName=None
      resourceObject=null
      tiling=NoTile
      imageSize=<0 0>
      margin=<ltrb 0 0 0 0>
      outlineSettings {
       bUseBrushTransparency=false
       roundingType=FixedRadius
       width=1
       color {colorUseRule=UseColor_ColorTable, specifiedColor=<rgba 0 0 0 0>}
       cornerRadii=<4 4 4 4>
      }
      tintColor {colorUseRule=UseColor_Specified, specifiedColor=<rgba 0.0048 0.0048 0.0048 1>}
      uVRegion {bIsValid=false, max=<0 0>, min=<0 0>}
     }
     backgroundImageHovered {
      bIsDynamicallyLoaded=false
      drawAs=RoundedBox
      imageType=NoImage
      mirroring=NoMirror
      resourceName=None
      resourceObject=null
      tiling=NoTile
      imageSize=<0 0>
      margin=<ltrb 0 0 0 0>
      outlineSettings {
       bUseBrushTransparency=false
       roundingType=FixedRadius
       width=1
       color {colorUseRule=UseColor_ColorTable, specifiedColor=<rgba 0 0 0 0>}
       cornerRadii=<4 4 4 4>
      }
      tintColor {colorUseRule=UseColor_Specified, specifiedColor=<rgba 0.0048 0.0048 0.0048 1>}
      uVRegion {bIsValid=false, max=<0 0>, min=<0 0>}
     }
     backgroundImageNormal {
      bIsDynamicallyLoaded=false
      drawAs=RoundedBox
      imageType=NoImage
      mirroring=NoMirror
      resourceName=None
      resourceObject=null
      tiling=NoTile
      imageSize=<0 0>
      margin=<ltrb 0 0 0 0>
      outlineSettings {
       bUseBrushTransparency=false
       roundingType=FixedRadius
       width=1
       color {colorUseRule=UseColor_ColorTable, specifiedColor=<rgba 0 0 0 0>}
       cornerRadii=<4 4 4 4>
      }
      tintColor {colorUseRule=UseColor_Specified, specifiedColor=<rgba 0.0048 0.0048 0.0048 1>}
      uVRegion {bIsValid=false, max=<0 0>, min=<0 0>}
     }
     backgroundImageReadOnly {
      bIsDynamicallyLoaded=false
      drawAs=RoundedBox
      imageType=NoImage
      mirroring=NoMirror
      resourceName=None
      resourceObject=null
      tiling=NoTile
      imageSize=<0 0>
      margin=<ltrb 0 0 0 0>
      outlineSettings {
       bUseBrushTransparency=false
       roundingType=FixedRadius
       width=1
       color {colorUseRule=UseColor_ColorTable, specifiedColor=<rgba 0 0 0 0>}
       cornerRadii=<4 4 4 4>
      }
      tintColor {colorUseRule=UseColor_Specified, specifiedColor=<rgba 0.0284 0.0284 0.0284 1>}
      uVRegion {bIsValid=false, max=<0 0>, min=<0 0>}
     }
     focusedForegroundColor {colorUseRule=UseColor_Specified, specifiedColor=<rgba 1 1 1 1>}
     foregroundColor {colorUseRule=UseColor_Specified, specifiedColor=<rgba 0.5271 0.5271 0.5271 1>}
     hScrollBarPadding=<ltrb 0 0 0 0>
     padding=<ltrb 8 3 350 4>
     readOnlyForegroundColor {colorUseRule=UseColor_Foreground, specifiedColor=<rgba 1 0 1 1>}
     scrollBarStyle {
      thickness=8
      draggedThumbImage {
       bIsDynamicallyLoaded=false
       drawAs=RoundedBox
       imageType=NoImage
       mirroring=NoMirror
       resourceName=None
       resourceObject=null
       tiling=NoTile
       imageSize=<0 0>
       margin=<ltrb 0 0 0 0>
       outlineSettings {
        bUseBrushTransparency=false
        roundingType=FixedRadius
        width=0
        color {colorUseRule=UseColor_Specified, specifiedColor=<rgba 0 0 0 0>}
        cornerRadii=<4 4 4 4>
       }
       tintColor {colorUseRule=UseColor_Specified, specifiedColor=<rgba 0.2159 0.2159 0.2159 1>}
       uVRegion {bIsValid=false, max=<0 0>, min=<0 0>}
      }
      horizontalBackgroundImage {
       bIsDynamicallyLoaded=false
       drawAs=NoDrawType
       imageType=NoImage
       mirroring=NoMirror
       resourceName=None
       resourceObject=null
       tiling=NoTile
       imageSize=<0 0>
       margin=<ltrb 0 0 0 0>
       outlineSettings {
        bUseBrushTransparency=false
        roundingType=HalfHeightRadius
        width=0
        color {colorUseRule=UseColor_Specified, specifiedColor=<rgba 0 0 0 0>}
        cornerRadii=<0 0 0 0>
       }
       tintColor {colorUseRule=UseColor_Specified, specifiedColor=<rgba 1 1 1 1>}
       uVRegion {bIsValid=false, max=<0 0>, min=<0 0>}
      }
      horizontalBottomSlotImage {
       bIsDynamicallyLoaded=false
       drawAs=NoDrawType
       imageType=NoImage
       mirroring=NoMirror
       resourceName=None
       resourceObject=null
       tiling=NoTile
       imageSize=<0 0>
       margin=<ltrb 0 0 0 0>
       outlineSettings {
        bUseBrushTransparency=false
        roundingType=HalfHeightRadius
        width=0
        color {colorUseRule=UseColor_Specified, specifiedColor=<rgba 0 0 0 0>}
        cornerRadii=<0 0 0 0>
       }
       tintColor {colorUseRule=UseColor_Specified, specifiedColor=<rgba 1 1 1 1>}
       uVRegion {bIsValid=false, max=<0 0>, min=<0 0>}
      }
      horizontalTopSlotImage {
       bIsDynamicallyLoaded=false
       drawAs=NoDrawType
       imageType=NoImage
       mirroring=NoMirror
       resourceName=None
       resourceObject=null
       tiling=NoTile
       imageSize=<0 0>
       margin=<ltrb 0 0 0 0>
       outlineSettings {
        bUseBrushTransparency=false
        roundingType=HalfHeightRadius
        width=0
        color {colorUseRule=UseColor_Specified, specifiedColor=<rgba 0 0 0 0>}
        cornerRadii=<0 0 0 0>
       }
       tintColor {colorUseRule=UseColor_Specified, specifiedColor=<rgba 1 1 1 1>}
       uVRegion {bIsValid=false, max=<0 0>, min=<0 0>}
      }
      hoveredThumbImage {
       bIsDynamicallyLoaded=false
       drawAs=RoundedBox
       imageType=NoImage
       mirroring=NoMirror
       resourceName=None
       resourceObject=null
       tiling=NoTile
       imageSize=<0 0>
       margin=<ltrb 0 0 0 0>
       outlineSettings {
        bUseBrushTransparency=false
        roundingType=FixedRadius
        width=0
        color {colorUseRule=UseColor_Specified, specifiedColor=<rgba 0 0 0 0>}
        cornerRadii=<4 4 4 4>
       }
       tintColor {colorUseRule=UseColor_Specified, specifiedColor=<rgba 0.2159 0.2159 0.2159 1>}
       uVRegion {bIsValid=false, max=<0 0>, min=<0 0>}
      }
      normalThumbImage {
       bIsDynamicallyLoaded=false
       drawAs=RoundedBox
       imageType=NoImage
       mirroring=NoMirror
       resourceName=None
       resourceObject=null
       tiling=NoTile
       imageSize=<0 0>
       margin=<ltrb 0 0 0 0>
       outlineSettings {
        bUseBrushTransparency=false
        roundingType=FixedRadius
        width=0
        color {colorUseRule=UseColor_Specified, specifiedColor=<rgba 0 0 0 0>}
        cornerRadii=<4 4 4 4>
       }
       tintColor {colorUseRule=UseColor_Specified, specifiedColor=<rgba 0.0953 0.0953 0.0953 1>}
       uVRegion {bIsValid=false, max=<0 0>, min=<0 0>}
      }
      verticalBackgroundImage {
       bIsDynamicallyLoaded=false
       drawAs=NoDrawType
       imageType=NoImage
       mirroring=NoMirror
       resourceName=None
       resourceObject=null
       tiling=NoTile
       imageSize=<0 0>
       margin=<ltrb 0 0 0 0>
       outlineSettings {
        bUseBrushTransparency=false
        roundingType=HalfHeightRadius
        width=0
        color {colorUseRule=UseColor_Specified, specifiedColor=<rgba 0 0 0 0>}
        cornerRadii=<0 0 0 0>
       }
       tintColor {colorUseRule=UseColor_Specified, specifiedColor=<rgba 1 1 1 1>}
       uVRegion {bIsValid=false, max=<0 0>, min=<0 0>}
      }
      verticalBottomSlotImage {
       bIsDynamicallyLoaded=false
       drawAs=NoDrawType
       imageType=NoImage
       mirroring=NoMirror
       resourceName=None
       resourceObject=null
       tiling=NoTile
       imageSize=<0 0>
       margin=<ltrb 0 0 0 0>
       outlineSettings {
        bUseBrushTransparency=false
        roundingType=HalfHeightRadius
        width=0
        color {colorUseRule=UseColor_Specified, specifiedColor=<rgba 0 0 0 0>}
        cornerRadii=<0 0 0 0>
       }
       tintColor {colorUseRule=UseColor_Specified, specifiedColor=<rgba 1 1 1 1>}
       uVRegion {bIsValid=false, max=<0 0>, min=<0 0>}
      }
      verticalTopSlotImage {
       bIsDynamicallyLoaded=false
       drawAs=NoDrawType
       imageType=NoImage
       mirroring=NoMirror
       resourceName=None
       resourceObject=null
       tiling=NoTile
       imageSize=<0 0>
       margin=<ltrb 0 0 0 0>
       outlineSettings {
        bUseBrushTransparency=false
        roundingType=HalfHeightRadius
        width=0
        color {colorUseRule=UseColor_Specified, specifiedColor=<rgba 0 0 0 0>}
        cornerRadii=<0 0 0 0>
       }
       tintColor {colorUseRule=UseColor_Specified, specifiedColor=<rgba 1 1 1 1>}
       uVRegion {bIsValid=false, max=<0 0>, min=<0 0>}
      }
     }
     textStyle {
      overflowPolicy=Clip
      transformPolicy=None
      colorAndOpacity {colorUseRule=UseColor_Foreground, specifiedColor=<rgba 1 0 1 1>}
      font {
       bForceMonospaced=false
       bMaterialIsStencil=false
       fontMaterial=null
       fontObject=null
       letterSpacing=0
       monospacedWidth=1
       size=10
       skewAmount=0
       typefaceFontName=Regular
       outlineSettings {
        bApplyOutlineToDropShadows=false
        bMiteredCorners=false
        bSeparateFillAlpha=false
        outlineMaterial=null
        outlineSize=0
        outlineColor=<rgba 0 0 0 1>
       }
      }
      highlightColor {colorUseRule=UseColor_ColorTable, specifiedColor=<rgba 0 0 0 0>}
      highlightShape {
       bIsDynamicallyLoaded=false
       drawAs=Image
       imageType=NoImage
       mirroring=NoMirror
       resourceName=None
       resourceObject=null
       tiling=NoTile
       imageSize=<0 0>
       margin=<ltrb 0 0 0 0>
       outlineSettings {
        bUseBrushTransparency=false
        roundingType=HalfHeightRadius
        width=0
        color {colorUseRule=UseColor_Specified, specifiedColor=<rgba 0 0 0 0>}
        cornerRadii=<0 0 0 0>
       }
       tintColor {colorUseRule=UseColor_ColorTable, specifiedColor=<rgba 0 0 0 0>}
       uVRegion {bIsValid=false, max=<0 0>, min=<0 0>}
      }
      selectedBackgroundColor {colorUseRule=UseColor_ColorTable, specifiedColor=<rgba 0 0 0 0>}
      shadowColorAndOpacity=<rgba 0 0 0 1>
      shadowOffset=<0 0>
      strikeBrush {
       bIsDynamicallyLoaded=false
       drawAs=Image
       imageType=NoImage
       mirroring=NoMirror
       resourceName=None
       resourceObject=null
       tiling=NoTile
       imageSize=<32 32>
       margin=<ltrb 0 0 0 0>
       outlineSettings {
        bUseBrushTransparency=false
        roundingType=HalfHeightRadius
        width=0
        color {colorUseRule=UseColor_Specified, specifiedColor=<rgba 0 0 0 0>}
        cornerRadii=<0 0 0 0>
       }
       tintColor {colorUseRule=UseColor_Specified, specifiedColor=<rgba 1 1 1 1>}
       uVRegion {bIsValid=false, max=<0 0>, min=<0 0>}
      }
      underlineBrush {
       bIsDynamicallyLoaded=false
       drawAs=Image
       imageType=NoImage
       mirroring=NoMirror
       resourceName=None
       resourceObject=null
       tiling=NoTile
       imageSize=<32 32>
       margin=<ltrb 0 0 0 0>
       outlineSettings {
        bUseBrushTransparency=false
        roundingType=HalfHeightRadius
        width=0
        color {colorUseRule=UseColor_Specified, specifiedColor=<rgba 0 0 0 0>}
        cornerRadii=<0 0 0 0>
       }
       tintColor {colorUseRule=UseColor_Specified, specifiedColor=<rgba 1 1 1 1>}
       uVRegion {bIsValid=false, max=<0 0>, min=<0 0>}
      }
     }
     vScrollBarPadding=<ltrb 0 0 0 0>
    }}
   slot {class=GridSlot, properties {column=1}}
  }
  UEFN_TextBlock_C {
   widgetClassPath=&UEFN_TextBlock
   parent=GridPanelAssetPathName
   widget=UEFN_TextBlock_C
   properties {
    text=Path:
    font {
     bForceMonospaced=false
     bMaterialIsStencil=false
     fontMaterial=null
     fontObject=&Roboto
     letterSpacing=0
     monospacedWidth=1
     size=19.5
     skewAmount=0
     typefaceFontName=Regular
     outlineSettings {
      bApplyOutlineToDropShadows=false
      bMiteredCorners=false
      bSeparateFillAlpha=false
      outlineMaterial=null
      outlineSize=0
      outlineColor=<rgba 0 0 0 1>
     }
    }
   }
   slot {class=GridSlot, properties {horizontalAlignment=HAlign_Right, verticalAlignment=VAlign_Center}}
  }
  AssetNameInput {
   widgetClassPath=&EditorUtilityEditableTextBox
   bIsVariable=true
   parent=GridPanelAssetPathName
   widget=AssetNameInput
   properties {
    toolTipText=What is the name of he product
    widgetStyle {
     backgroundColor {colorUseRule=UseColor_Specified, specifiedColor=<rgba 1 1 1 1>}
     backgroundImageFocused {
      bIsDynamicallyLoaded=false
      drawAs=RoundedBox
      imageType=NoImage
      mirroring=NoMirror
      resourceName=None
      resourceObject=null
      tiling=NoTile
      imageSize=<0 0>
      margin=<ltrb 0 0 0 0>
      outlineSettings {
       bUseBrushTransparency=false
       roundingType=FixedRadius
       width=1
       color {colorUseRule=UseColor_ColorTable, specifiedColor=<rgba 0 0 0 0>}
       cornerRadii=<4 4 4 4>
      }
      tintColor {colorUseRule=UseColor_Specified, specifiedColor=<rgba 0.0048 0.0048 0.0048 1>}
      uVRegion {bIsValid=false, max=<0 0>, min=<0 0>}
     }
     backgroundImageHovered {
      bIsDynamicallyLoaded=false
      drawAs=RoundedBox
      imageType=NoImage
      mirroring=NoMirror
      resourceName=None
      resourceObject=null
      tiling=NoTile
      imageSize=<0 0>
      margin=<ltrb 0 0 0 0>
      outlineSettings {
       bUseBrushTransparency=false
       roundingType=FixedRadius
       width=1
       color {colorUseRule=UseColor_ColorTable, specifiedColor=<rgba 0 0 0 0>}
       cornerRadii=<4 4 4 4>
      }
      tintColor {colorUseRule=UseColor_Specified, specifiedColor=<rgba 0.0048 0.0048 0.0048 1>}
      uVRegion {bIsValid=false, max=<0 0>, min=<0 0>}
     }
     backgroundImageNormal {
      bIsDynamicallyLoaded=false
      drawAs=RoundedBox
      imageType=NoImage
      mirroring=NoMirror
      resourceName=None
      resourceObject=null
      tiling=NoTile
      imageSize=<0 0>
      margin=<ltrb 0 0 0 0>
      outlineSettings {
       bUseBrushTransparency=false
       roundingType=FixedRadius
       width=1
       color {colorUseRule=UseColor_ColorTable, specifiedColor=<rgba 0 0 0 0>}
       cornerRadii=<4 4 4 4>
      }
      tintColor {colorUseRule=UseColor_Specified, specifiedColor=<rgba 0.0048 0.0048 0.0048 1>}
      uVRegion {bIsValid=false, max=<0 0>, min=<0 0>}
     }
     backgroundImageReadOnly {
      bIsDynamicallyLoaded=false
      drawAs=RoundedBox
      imageType=NoImage
      mirroring=NoMirror
      resourceName=None
      resourceObject=null
      tiling=NoTile
      imageSize=<0 0>
      margin=<ltrb 0 0 0 0>
      outlineSettings {
       bUseBrushTransparency=false
       roundingType=FixedRadius
       width=1
       color {colorUseRule=UseColor_ColorTable, specifiedColor=<rgba 0 0 0 0>}
       cornerRadii=<4 4 4 4>
      }
      tintColor {colorUseRule=UseColor_Specified, specifiedColor=<rgba 0.0284 0.0284 0.0284 1>}
      uVRegion {bIsValid=false, max=<0 0>, min=<0 0>}
     }
     focusedForegroundColor {colorUseRule=UseColor_Specified, specifiedColor=<rgba 1 1 1 1>}
     foregroundColor {colorUseRule=UseColor_Specified, specifiedColor=<rgba 0.5271 0.5271 0.5271 1>}
     hScrollBarPadding=<ltrb 0 0 0 0>
     padding=<ltrb 8 3 350 4>
     readOnlyForegroundColor {colorUseRule=UseColor_Foreground, specifiedColor=<rgba 1 0 1 1>}
     scrollBarStyle {
      thickness=8
      draggedThumbImage {
       bIsDynamicallyLoaded=false
       drawAs=RoundedBox
       imageType=NoImage
       mirroring=NoMirror
       resourceName=None
       resourceObject=null
       tiling=NoTile
       imageSize=<0 0>
       margin=<ltrb 0 0 0 0>
       outlineSettings {
        bUseBrushTransparency=false
        roundingType=FixedRadius
        width=0
        color {colorUseRule=UseColor_Specified, specifiedColor=<rgba 0 0 0 0>}
        cornerRadii=<4 4 4 4>
       }
       tintColor {colorUseRule=UseColor_Specified, specifiedColor=<rgba 0.2159 0.2159 0.2159 1>}
       uVRegion {bIsValid=false, max=<0 0>, min=<0 0>}
      }
      horizontalBackgroundImage {
       bIsDynamicallyLoaded=false
       drawAs=NoDrawType
       imageType=NoImage
       mirroring=NoMirror
       resourceName=None
       resourceObject=null
       tiling=NoTile
       imageSize=<0 0>
       margin=<ltrb 0 0 0 0>
       outlineSettings {
        bUseBrushTransparency=false
        roundingType=HalfHeightRadius
        width=0
        color {colorUseRule=UseColor_Specified, specifiedColor=<rgba 0 0 0 0>}
        cornerRadii=<0 0 0 0>
       }
       tintColor {colorUseRule=UseColor_Specified, specifiedColor=<rgba 1 1 1 1>}
       uVRegion {bIsValid=false, max=<0 0>, min=<0 0>}
      }
      horizontalBottomSlotImage {
       bIsDynamicallyLoaded=false
       drawAs=NoDrawType
       imageType=NoImage
       mirroring=NoMirror
       resourceName=None
       resourceObject=null
       tiling=NoTile
       imageSize=<0 0>
       margin=<ltrb 0 0 0 0>
       outlineSettings {
        bUseBrushTransparency=false
        roundingType=HalfHeightRadius
        width=0
        color {colorUseRule=UseColor_Specified, specifiedColor=<rgba 0 0 0 0>}
        cornerRadii=<0 0 0 0>
       }
       tintColor {colorUseRule=UseColor_Specified, specifiedColor=<rgba 1 1 1 1>}
       uVRegion {bIsValid=false, max=<0 0>, min=<0 0>}
      }
      horizontalTopSlotImage {
       bIsDynamicallyLoaded=false
       drawAs=NoDrawType
       imageType=NoImage
       mirroring=NoMirror
       resourceName=None
       resourceObject=null
       tiling=NoTile
       imageSize=<0 0>
       margin=<ltrb 0 0 0 0>
       outlineSettings {
        bUseBrushTransparency=false
        roundingType=HalfHeightRadius
        width=0
        color {colorUseRule=UseColor_Specified, specifiedColor=<rgba 0 0 0 0>}
        cornerRadii=<0 0 0 0>
       }
       tintColor {colorUseRule=UseColor_Specified, specifiedColor=<rgba 1 1 1 1>}
       uVRegion {bIsValid=false, max=<0 0>, min=<0 0>}
      }
      hoveredThumbImage {
       bIsDynamicallyLoaded=false
       drawAs=RoundedBox
       imageType=NoImage
       mirroring=NoMirror
       resourceName=None
       resourceObject=null
       tiling=NoTile
       imageSize=<0 0>
       margin=<ltrb 0 0 0 0>
       outlineSettings {
        bUseBrushTransparency=false
        roundingType=FixedRadius
        width=0
        color {colorUseRule=UseColor_Specified, specifiedColor=<rgba 0 0 0 0>}
        cornerRadii=<4 4 4 4>
       }
       tintColor {colorUseRule=UseColor_Specified, specifiedColor=<rgba 0.2159 0.2159 0.2159 1>}
       uVRegion {bIsValid=false, max=<0 0>, min=<0 0>}
      }
      normalThumbImage {
       bIsDynamicallyLoaded=false
       drawAs=RoundedBox
       imageType=NoImage
       mirroring=NoMirror
       resourceName=None
       resourceObject=null
       tiling=NoTile
       imageSize=<0 0>
       margin=<ltrb 0 0 0 0>
       outlineSettings {
        bUseBrushTransparency=false
        roundingType=FixedRadius
        width=0
        color {colorUseRule=UseColor_Specified, specifiedColor=<rgba 0 0 0 0>}
        cornerRadii=<4 4 4 4>
       }
       tintColor {colorUseRule=UseColor_Specified, specifiedColor=<rgba 0.0953 0.0953 0.0953 1>}
       uVRegion {bIsValid=false, max=<0 0>, min=<0 0>}
      }
      verticalBackgroundImage {
       bIsDynamicallyLoaded=false
       drawAs=NoDrawType
       imageType=NoImage
       mirroring=NoMirror
       resourceName=None
       resourceObject=null
       tiling=NoTile
       imageSize=<0 0>
       margin=<ltrb 0 0 0 0>
       outlineSettings {
        bUseBrushTransparency=false
        roundingType=HalfHeightRadius
        width=0
        color {colorUseRule=UseColor_Specified, specifiedColor=<rgba 0 0 0 0>}
        cornerRadii=<0 0 0 0>
       }
       tintColor {colorUseRule=UseColor_Specified, specifiedColor=<rgba 1 1 1 1>}
       uVRegion {bIsValid=false, max=<0 0>, min=<0 0>}
      }
      verticalBottomSlotImage {
       bIsDynamicallyLoaded=false
       drawAs=NoDrawType
       imageType=NoImage
       mirroring=NoMirror
       resourceName=None
       resourceObject=null
       tiling=NoTile
       imageSize=<0 0>
       margin=<ltrb 0 0 0 0>
       outlineSettings {
        bUseBrushTransparency=false
        roundingType=HalfHeightRadius
        width=0
        color {colorUseRule=UseColor_Specified, specifiedColor=<rgba 0 0 0 0>}
        cornerRadii=<0 0 0 0>
       }
       tintColor {colorUseRule=UseColor_Specified, specifiedColor=<rgba 1 1 1 1>}
       uVRegion {bIsValid=false, max=<0 0>, min=<0 0>}
      }
      verticalTopSlotImage {
       bIsDynamicallyLoaded=false
       drawAs=NoDrawType
       imageType=NoImage
       mirroring=NoMirror
       resourceName=None
       resourceObject=null
       tiling=NoTile
       imageSize=<0 0>
       margin=<ltrb 0 0 0 0>
       outlineSettings {
        bUseBrushTransparency=false
        roundingType=HalfHeightRadius
        width=0
        color {colorUseRule=UseColor_Specified, specifiedColor=<rgba 0 0 0 0>}
        cornerRadii=<0 0 0 0>
       }
       tintColor {colorUseRule=UseColor_Specified, specifiedColor=<rgba 1 1 1 1>}
       uVRegion {bIsValid=false, max=<0 0>, min=<0 0>}
      }
     }
     textStyle {
      overflowPolicy=Clip
      transformPolicy=None
      colorAndOpacity {colorUseRule=UseColor_Foreground, specifiedColor=<rgba 1 0 1 1>}
      font {
       bForceMonospaced=false
       bMaterialIsStencil=false
       fontMaterial=null
       fontObject=null
       letterSpacing=0
       monospacedWidth=1
       size=10
       skewAmount=0
       typefaceFontName=Regular
       outlineSettings {
        bApplyOutlineToDropShadows=false
        bMiteredCorners=false
        bSeparateFillAlpha=false
        outlineMaterial=null
        outlineSize=0
        outlineColor=<rgba 0 0 0 1>
       }
      }
      highlightColor {colorUseRule=UseColor_ColorTable, specifiedColor=<rgba 0 0 0 0>}
      highlightShape {
       bIsDynamicallyLoaded=false
       drawAs=Image
       imageType=NoImage
       mirroring=NoMirror
       resourceName=None
       resourceObject=null
       tiling=NoTile
       imageSize=<0 0>
       margin=<ltrb 0 0 0 0>
       outlineSettings {
        bUseBrushTransparency=false
        roundingType=HalfHeightRadius
        width=0
        color {colorUseRule=UseColor_Specified, specifiedColor=<rgba 0 0 0 0>}
        cornerRadii=<0 0 0 0>
       }
       tintColor {colorUseRule=UseColor_ColorTable, specifiedColor=<rgba 0 0 0 0>}
       uVRegion {bIsValid=false, max=<0 0>, min=<0 0>}
      }
      selectedBackgroundColor {colorUseRule=UseColor_ColorTable, specifiedColor=<rgba 0 0 0 0>}
      shadowColorAndOpacity=<rgba 0 0 0 1>
      shadowOffset=<0 0>
      strikeBrush {
       bIsDynamicallyLoaded=false
       drawAs=Image
       imageType=NoImage
       mirroring=NoMirror
       resourceName=None
       resourceObject=null
       tiling=NoTile
       imageSize=<32 32>
       margin=<ltrb 0 0 0 0>
       outlineSettings {
        bUseBrushTransparency=false
        roundingType=HalfHeightRadius
        width=0
        color {colorUseRule=UseColor_Specified, specifiedColor=<rgba 0 0 0 0>}
        cornerRadii=<0 0 0 0>
       }
       tintColor {colorUseRule=UseColor_Specified, specifiedColor=<rgba 1 1 1 1>}
       uVRegion {bIsValid=false, max=<0 0>, min=<0 0>}
      }
      underlineBrush {
       bIsDynamicallyLoaded=false
       drawAs=Image
       imageType=NoImage
       mirroring=NoMirror
       resourceName=None
       resourceObject=null
       tiling=NoTile
       imageSize=<32 32>
       margin=<ltrb 0 0 0 0>
       outlineSettings {
        bUseBrushTransparency=false
        roundingType=HalfHeightRadius
        width=0
        color {colorUseRule=UseColor_Specified, specifiedColor=<rgba 0 0 0 0>}
        cornerRadii=<0 0 0 0>
       }
       tintColor {colorUseRule=UseColor_Specified, specifiedColor=<rgba 1 1 1 1>}
       uVRegion {bIsValid=false, max=<0 0>, min=<0 0>}
      }
     }
     vScrollBarPadding=<ltrb 0 0 0 0>
    }
   }
   slot {class=GridSlot, properties {column=1, row=1}}
  }
  EditorUtilityButton_HelpButton {
   widgetClassPath=&EditorUtilityButton
   parent=GridPanelAssetPathName
   widget=EditorUtilityButton_HelpButton
   properties {toolTipText=Make metasounds from selected soundwaves, backgroundColor=<rgba 0.022 0.3668 1 1>}
   slot {class=GridSlot, properties {column=2}}
  }
  UEFN_TextBlock_C_12 {
   widgetClassPath=&UEFN_TextBlock
   parent=EditorUtilityButton_HelpButton
   widget=UEFN_TextBlock_C_12
   properties {
    justification=Center
    text=Help/Docs
    font {
     bForceMonospaced=false
     bMaterialIsStencil=false
     fontMaterial=null
     fontObject=&BurbankSmall
     letterSpacing=0
     monospacedWidth=1
     size=9
     skewAmount=0
     typefaceFontName=Black
     outlineSettings {
      bApplyOutlineToDropShadows=false
      bMiteredCorners=false
      bSeparateFillAlpha=false
      outlineMaterial=null
      outlineSize=0
      outlineColor=<rgba 0 0 0 1>
     }
    }
   }
   slot {class=ButtonSlot}
  }
  HorizontalBoxWidgetSwitchSelectionButtons {
   widgetClassPath=&HorizontalBox
   parent=UIKitGridPanel_616
   widget=HorizontalBoxWidgetSwitchSelectionButtons
   slot {class=GridSlot, properties {horizontalAlignment=HAlign_Center, row=2, padding=<ltrb 0 0 0 10>}}
  }
  EditorUtilityButton_Authoring {
   widgetClassPath=&EditorUtilityButton
   parent=HorizontalBoxWidgetSwitchSelectionButtons
   widget=EditorUtilityButton_Authoring
   properties {toolTipText=Make metasounds from selected soundwaves, backgroundColor=<rgba 0 0.3845 1 1>}
   slot {class=HorizontalBoxSlot}
  }
  UEFN_TextBlock_C_9 {
   widgetClassPath=&UEFN_TextBlock
   parent=EditorUtilityButton_Authoring
   widget=UEFN_TextBlock_C_9
   properties {
    justification=Center
    text=Authoring
    font {
     bForceMonospaced=false
     bMaterialIsStencil=false
     fontMaterial=null
     fontObject=&BurbankSmall
     letterSpacing=0
     monospacedWidth=1
     size=12
     skewAmount=0
     typefaceFontName=Black
     outlineSettings {
      bApplyOutlineToDropShadows=false
      bMiteredCorners=false
      bSeparateFillAlpha=false
      outlineMaterial=null
      outlineSize=0
      outlineColor=<rgba 0 0 0 1>
     }
    }
   }
   slot {class=ButtonSlot}
  }
  EditorUtilityButton_PreviewButton {
   widgetClassPath=&EditorUtilityButton
   parent=HorizontalBoxWidgetSwitchSelectionButtons
   widget=EditorUtilityButton_PreviewButton
   properties {toolTipText=Make metasounds from selected soundwaves, backgroundColor=<rgba 0 0.3845 1 1>}
   slot {class=HorizontalBoxSlot}
  }
  UEFN_TextBlock_C_1 {
   widgetClassPath=&UEFN_TextBlock
   parent=EditorUtilityButton_PreviewButton
   widget=UEFN_TextBlock_C_1
   properties {
    justification=Center
    text=Previewing
    font {
     bForceMonospaced=false
     bMaterialIsStencil=false
     fontMaterial=null
     fontObject=&BurbankSmall
     letterSpacing=0
     monospacedWidth=1
     size=12
     skewAmount=0
     typefaceFontName=Black
     outlineSettings {
      bApplyOutlineToDropShadows=false
      bMiteredCorners=false
      bSeparateFillAlpha=false
      outlineMaterial=null
      outlineSize=0
      outlineColor=<rgba 0 0 0 1>
     }
    }
   }
   slot {class=ButtonSlot}
  }
  EditorUtilityButton_Loudness {
   widgetClassPath=&EditorUtilityButton
   parent=HorizontalBoxWidgetSwitchSelectionButtons
   widget=EditorUtilityButton_Loudness
   properties {toolTipText=Make metasounds from selected soundwaves, backgroundColor=<rgba 0 0.3845 1 1>}
   slot {class=HorizontalBoxSlot}
  }
  UEFN_TextBlock_C_10 {
   widgetClassPath=&UEFN_TextBlock
   parent=EditorUtilityButton_Loudness
   widget=UEFN_TextBlock_C_10
   properties {
    justification=Center
    text=Loudness
    font {
     bForceMonospaced=false
     bMaterialIsStencil=false
     fontMaterial=null
     fontObject=&BurbankSmall
     letterSpacing=0
     monospacedWidth=1
     size=12
     skewAmount=0
     typefaceFontName=Black
     outlineSettings {
      bApplyOutlineToDropShadows=false
      bMiteredCorners=false
      bSeparateFillAlpha=false
      outlineMaterial=null
      outlineSize=0
      outlineColor=<rgba 0 0 0 1>
     }
    }
   }
   slot {class=ButtonSlot}
  }
  CommonActivatableWidgetSwitcher_171 {
   widgetClassPath=&CommonActivatableWidgetSwitcher
   bIsVariable=true
   parent=UIKitGridPanel_616
   widget=CommonActivatableWidgetSwitcher_171
   slot {class=GridSlot, properties {row=3}}
  }
  UIKitGridPanelAuthoring {
   widgetClassPath=&UIKitGridPanel
   parent=CommonActivatableWidgetSwitcher_171
   widget=UIKitGridPanelAuthoring
   slot {class=WidgetSwitcherSlot, properties {horizontalAlignment=HAlign_Center, verticalAlignment=VAlign_Center}}
  }
  UIKitGridPanelModelValues {
   widgetClassPath=&UIKitGridPanel
   parent=UIKitGridPanelAuthoring
   widget=UIKitGridPanelModelValues
   slot {class=GridSlot, properties {horizontalAlignment=HAlign_Left, verticalAlignment=VAlign_Top}}
  }
  TextWeight_8 {
   widgetClassPath=&TextBlock
   parent=UIKitGridPanelModelValues
   widget=TextWeight_8
   properties {
    text=Physics
    font {
     bForceMonospaced=false
     bMaterialIsStencil=false
     fontMaterial=null
     fontObject=&Roboto
     letterSpacing=0
     monospacedWidth=1
     size=22.5
     skewAmount=0
     typefaceFontName=Italic
     outlineSettings {
      bApplyOutlineToDropShadows=false
      bMiteredCorners=false
      bSeparateFillAlpha=false
      outlineMaterial=null
      outlineSize=0
      outlineColor=<rgba 0 0 0 1>
     }
    }
   }
   slot {class=GridSlot, properties {horizontalAlignment=HAlign_Left, verticalAlignment=VAlign_Center}}
  }
  SinglePropertyView_EngineSound_Weight {
   widgetClassPath=&SinglePropertyView
   parent=UIKitGridPanelModelValues
   widget=SinglePropertyView_EngineSound_Weight
   properties {propertyName=Weight}
   slot {class=GridSlot, properties {row=4}}
  }
  SinglePropertyView_EngineSound_Torque {
   widgetClassPath=&SinglePropertyView
   parent=UIKitGridPanelModelValues
   widget=SinglePropertyView_EngineSound_Torque
   properties {propertyName=Torque}
   slot {class=GridSlot, properties {row=5}}
  }
  SinglePropertyView_EngineSound_BrakingHorsepower {
   widgetClassPath=&SinglePropertyView
   parent=UIKitGridPanelModelValues
   widget=SinglePropertyView_EngineSound_BrakingHorsepower
   properties {propertyName=Braking Horsepower}
   slot {class=GridSlot, properties {row=6}}
  }
  UIKitGridPanelREVSettings {
   widgetClassPath=&UIKitGridPanel
   parent=UIKitGridPanelAuthoring
   widget=UIKitGridPanelREVSettings
   slot {class=GridSlot, properties {horizontalAlignment=HAlign_Center, verticalAlignment=VAlign_Top, nudge=<0 123.7341>}}
  }
  TextWeight_3 {
   widgetClassPath=&TextBlock
   parent=UIKitGridPanelREVSettings
   widget=TextWeight_3
   properties {
    text=REV Settings
    font {
     bForceMonospaced=false
     bMaterialIsStencil=false
     fontMaterial=null
     fontObject=&Roboto
     letterSpacing=0
     monospacedWidth=1
     size=22.5
     skewAmount=0
     typefaceFontName=Italic
     outlineSettings {
      bApplyOutlineToDropShadows=false
      bMiteredCorners=false
      bSeparateFillAlpha=false
      outlineMaterial=null
      outlineSize=0
      outlineColor=<rgba 0 0 0 1>
     }
    }
   }
   slot {class=GridSlot, properties {horizontalAlignment=HAlign_Left, verticalAlignment=VAlign_Center}}
  }
  SinglePropertyView_EngineSound_RevSettings {
   widgetClassPath=&SinglePropertyView
   parent=UIKitGridPanelREVSettings
   widget=SinglePropertyView_EngineSound_RevSettings
   properties {propertyName=RevSettingsAsset}
   slot {class=GridSlot, properties {row=4}}
  }
  REVSettingsBackground_Green {
   widgetClassPath=&Image
   bIsVariable=true
   parent=UIKitGridPanelREVSettings
   widget=REVSettingsBackground_Green
   properties {colorAndOpacity=<rgba 0.0026 1 0.1925 0>}
   slot {class=GridSlot, properties {row=4}}
  }
  REVSettingsBackground_Red {
   widgetClassPath=&Image
   bIsVariable=true
   parent=UIKitGridPanelREVSettings
   widget=REVSettingsBackground_Red
   properties {colorAndOpacity=<rgba 1 0 0.0343 1>}
   slot {class=GridSlot, properties {row=4}}
  }
  CanvasPanel_0 {widgetClassPath=&CanvasPanel, parent=UIKitGridPanelREVSettings, widget=CanvasPanel_0, slot {class=GridSlot, properties {row=5}}}
  EditorUtilityButton_56_ImportRevSettings {
   widgetClassPath=&EditorUtilityButton
   parent=CanvasPanel_0
   widget=EditorUtilityButton_56_ImportRevSettings
   properties {backgroundColor=<rgba 0.0012 0.0595 0.3231 1>}
   slot {class=CanvasPanelSlot, properties {layoutData {alignment=<0 0>, anchors {maximum=<0 0>, minimum=<0 0>}, offsets=<ltrb 0 0 238.4718 21.992>}}}
  }
  UEFN_TextBlock_C_11 {
   widgetClassPath=&UEFN_TextBlock
   parent=EditorUtilityButton_56_ImportRevSettings
   widget=UEFN_TextBlock_C_11
   properties {
    text=Apply settings from .model file...
    font {
     bForceMonospaced=false
     bMaterialIsStencil=false
     fontMaterial=null
     fontObject=&BurbankSmall
     letterSpacing=0
     monospacedWidth=1
     size=9
     skewAmount=0
     typefaceFontName=Black
     outlineSettings {
      bApplyOutlineToDropShadows=false
      bMiteredCorners=false
      bSeparateFillAlpha=false
      outlineMaterial=null
      outlineSize=0
      outlineColor=<rgba 0 0 0 1>
     }
    }
   }
   slot {class=ButtonSlot}
  }
  UIKitGridPanelAssets {
   widgetClassPath=&UIKitGridPanel
   parent=UIKitGridPanelAuthoring
   widget=UIKitGridPanelAssets
   slot {class=GridSlot, properties {column=4, padding=<ltrb 10 0 10 0>}}
  }
  TextWeight {
   widgetClassPath=&TextBlock
   parent=UIKitGridPanelAssets
   widget=TextWeight
   properties {
    text=Assets
    font {
     bForceMonospaced=false
     bMaterialIsStencil=false
     fontMaterial=null
     fontObject=&Roboto
     letterSpacing=0
     monospacedWidth=1
     size=22.5
     skewAmount=0
     typefaceFontName=Italic
     outlineSettings {
      bApplyOutlineToDropShadows=false
      bMiteredCorners=false
      bSeparateFillAlpha=false
      outlineMaterial=null
      outlineSize=0
      outlineColor=<rgba 0 0 0 1>
     }
    }
   }
   slot {class=GridSlot, properties {horizontalAlignment=HAlign_Left, verticalAlignment=VAlign_Center}}
  }
  HorizontalBox_0 {
   widgetClassPath=&HorizontalBox
   parent=UIKitGridPanelAssets
   widget=HorizontalBox_0
   slot {class=GridSlot, properties {horizontalAlignment=HAlign_Left, row=1, verticalAlignment=VAlign_Top}}
  }
  WBP_MuteSoloButton_Engine {
   widgetClassPath=&/WBP_MuteSoloButton
   bIsVariable=true
   parent=HorizontalBox_0
   widget=WBP_MuteSoloButton_Engine
   properties {targetSoloSound=Engine}
   slot {class=HorizontalBoxSlot, properties {verticalAlignment=VAlign_Center, padding=<ltrb 0 0 15 0>}}
  }
  SinglePropertyView_EngineSound {
   widgetClassPath=&SinglePropertyView
   parent=HorizontalBox_0
   widget=SinglePropertyView_EngineSound
   properties {propertyName=Engine Sound}
   slot {class=HorizontalBoxSlot}
  }
  SinglePropertyView_IsIP {
   widgetClassPath=&SinglePropertyView
   parent=UIKitGridPanelAssets
   widget=SinglePropertyView_IsIP
   properties {propertyName=Is IP, toolTipText="When true, the asset is marked up as a 3rd party vehicle. "}
   slot {class=GridSlot, properties {column=1, row=1}}
  }
  TextWeight_1 {
   widgetClassPath=&TextBlock
   parent=UIKitGridPanelAssets
   widget=TextWeight_1
   properties {
    text=Additional Data
    font {
     bForceMonospaced=false
     bMaterialIsStencil=false
     fontMaterial=null
     fontObject=&Roboto
     letterSpacing=0
     monospacedWidth=1
     size=18.75
     skewAmount=0
     typefaceFontName=Italic
     outlineSettings {
      bApplyOutlineToDropShadows=false
      bMiteredCorners=false
      bSeparateFillAlpha=false
      outlineMaterial=null
      outlineSize=0
      outlineColor=<rgba 0 0 0 1>
     }
    }
   }
   slot {class=GridSlot, properties {column=1, horizontalAlignment=HAlign_Left, verticalAlignment=VAlign_Center}}
  }
  Border_196 {
   widgetClassPath=&Border
   parent=UIKitGridPanelAssets
   widget=Border_196
   properties {
    background {
     bIsDynamicallyLoaded=false
     drawAs=Image
     imageType=NoImage
     mirroring=NoMirror
     resourceName=None
     resourceObject=null
     tiling=NoTile
     imageSize=<32 32>
     margin=<ltrb 0 0 0 0>
     outlineSettings {
      bUseBrushTransparency=false
      roundingType=HalfHeightRadius
      width=0
      color {colorUseRule=UseColor_Specified, specifiedColor=<rgba 0 0 0 0>}
      cornerRadii=<0 0 0 0>
     }
     tintColor {colorUseRule=UseColor_Specified, specifiedColor=<rgba 0.7226 0.9617 1 1>}
     uVRegion {bIsValid=false, max=<0 0>, min=<0 0>}
    }
    brushColor=<rgba 0.0208 0.0208 0.0208 1>
    padding=<ltrb 0 0 0 0>
   }
   slot {class=GridSlot, properties {column=1, row=2}}
  }
  VerticalBox_203 {
   widgetClassPath=&VerticalBox
   parent=Border_196
   widget=VerticalBox_203
   slot {class=BorderSlot, properties {padding=<ltrb 0 0 0 0>}}
  }
  TextBlock_97 {
   widgetClassPath=&TextBlock
   parent=VerticalBox_203
   widget=TextBlock_97
   properties {
    text=EV Settings
    font {
     bForceMonospaced=false
     bMaterialIsStencil=false
     fontMaterial=null
     fontObject=&Roboto
     letterSpacing=0
     monospacedWidth=1
     size=7.5
     skewAmount=0
     typefaceFontName=Italic
     outlineSettings {
      bApplyOutlineToDropShadows=false
      bMiteredCorners=false
      bSeparateFillAlpha=false
      outlineMaterial=null
      outlineSize=0
      outlineColor=<rgba 0 0 0 1>
     }
    }
   }
   slot {class=VerticalBoxSlot, properties {horizontalAlignment=HAlign_Center}}
  }
  SinglePropertyView_IgnoreShifting {
   widgetClassPath=&SinglePropertyView
   parent=VerticalBox_203
   widget=SinglePropertyView_IgnoreShifting
   properties {
    propertyName=Ignore Shifting
    toolTipText="When true, the vehicle does not generate RPM values, using speed as the primary driver of the engine. Primarily used for EVS. "
   }
   slot {class=VerticalBoxSlot}
  }
  SinglePropertyView_IgnoreExhaustPops {
   widgetClassPath=&SinglePropertyView
   parent=VerticalBox_203
   widget=SinglePropertyView_IgnoreExhaustPops
   properties {propertyName=Ignore Exhaust Pops, toolTipText="When true, the gear shifting events are not sent that fire exhaust pops. This is "}
   slot {class=VerticalBoxSlot}
  }
  HorizontalBox_1 {widgetClassPath=&HorizontalBox, parent=UIKitGridPanelAssets, widget=HorizontalBox_1, slot {class=GridSlot, properties {row=2}}}
  WBP_MuteSoloButton_Ignition {
   widgetClassPath=&WBP_MuteSoloButton
   bIsVariable=true
   parent=HorizontalBox_1
   widget=WBP_MuteSoloButton_Ignition
   properties {targetSoloSound=Ignition}
   slot {class=HorizontalBoxSlot, properties {verticalAlignment=VAlign_Center, padding=<ltrb 0 0 15 0>}}
  }
  SinglePropertyView_Ignition {
   widgetClassPath=&SinglePropertyView
   parent=HorizontalBox_1
   widget=SinglePropertyView_Ignition
   properties {propertyName=Engine Ignition Sound, toolTipText=Ignition}
   slot {class=HorizontalBoxSlot}
  }
  HorizontalBox_2 {widgetClassPath=&HorizontalBox, parent=UIKitGridPanelAssets, widget=HorizontalBox_2, slot {class=GridSlot, properties {row=3}}}
  WBP_MuteSoloButton_Off {
   widgetClassPath=&WBP_MuteSoloButton
   bIsVariable=true
   parent=HorizontalBox_2
   widget=WBP_MuteSoloButton_Off
   properties {targetSoloSound=Off}
   slot {class=HorizontalBoxSlot, properties {verticalAlignment=VAlign_Center, padding=<ltrb 0 0 15 0>}}
  }
  SinglePropertyView_Off {
   widgetClassPath=&SinglePropertyView
   parent=HorizontalBox_2
   widget=SinglePropertyView_Off
   properties {propertyName=Engine Off Sound}
   slot {class=HorizontalBoxSlot}
  }
  HorizontalBox_3 {widgetClassPath=&HorizontalBox, parent=UIKitGridPanelAssets, widget=HorizontalBox_3, slot {class=GridSlot, properties {row=4}}}
  WBP_MuteSoloButton_Horn {
   widgetClassPath=&WBP_MuteSoloButton
   bIsVariable=true
   parent=HorizontalBox_3
   widget=WBP_MuteSoloButton_Horn
   properties {targetSoloSound=Horn}
   slot {class=HorizontalBoxSlot, properties {verticalAlignment=VAlign_Center, padding=<ltrb 0 0 15 0>}}
  }
  SinglePropertyView_Horn {
   widgetClassPath=&SinglePropertyView
   parent=HorizontalBox_3
   widget=SinglePropertyView_Horn
   properties {propertyName=Horn Sound}
   slot {class=HorizontalBoxSlot}
  }
  UIKitGridPanelCreatedAssets {
   widgetClassPath=&UIKitGridPanel
   parent=UIKitGridPanelAuthoring
   widget=UIKitGridPanelCreatedAssets
   slot {class=GridSlot, properties {column=5, horizontalAlignment=HAlign_Center, verticalAlignment=VAlign_Top}}
  }
  TextWeight_2 {
   widgetClassPath=&TextBlock
   parent=UIKitGridPanelCreatedAssets
   widget=TextWeight_2
   properties {
    text=Data Assets
    font {
     bForceMonospaced=false
     bMaterialIsStencil=false
     fontMaterial=null
     fontObject=&Roboto
     letterSpacing=0
     monospacedWidth=1
     size=22.5
     skewAmount=0
     typefaceFontName=Italic
     outlineSettings {
      bApplyOutlineToDropShadows=false
      bMiteredCorners=false
      bSeparateFillAlpha=false
      outlineMaterial=null
      outlineSize=0
      outlineColor=<rgba 0 0 0 1>
     }
    }
   }
   slot {class=GridSlot, properties {horizontalAlignment=HAlign_Left, verticalAlignment=VAlign_Center}}
  }
  SinglePropertyView_EngineSound_ChooserTable {
   widgetClassPath=&SinglePropertyView
   parent=UIKitGridPanelCreatedAssets
   widget=SinglePropertyView_EngineSound_ChooserTable
   properties {propertyName=CosmeticChooserTableSoftPtr}
   slot {class=GridSlot, properties {row=5}}
  }
  SinglePropertyView_EngineSound_VehicleItemDataAsset {
   widgetClassPath=&SinglePropertyView
   parent=UIKitGridPanelCreatedAssets
   widget=SinglePropertyView_EngineSound_VehicleItemDataAsset
   properties {propertyName=VehicleEngineAudioItemDefinition}
   slot {class=GridSlot, properties {row=6}}
  }
  UIKitGridPanelAudioPreviewControls {
   widgetClassPath=&UIKitGridPanel
   parent=CommonActivatableWidgetSwitcher_171
   widget=UIKitGridPanelAudioPreviewControls
   slot {class=WidgetSwitcherSlot}
  }
  TextBlock_2 {
   widgetClassPath=&TextBlock
   parent=UIKitGridPanelAudioPreviewControls
   widget=TextBlock_2
   properties {
    text=Perspective
    font {
     bForceMonospaced=false
     bMaterialIsStencil=false
     fontMaterial=null
     fontObject=null
     letterSpacing=0
     monospacedWidth=1
     size=24
     skewAmount=0
     typefaceFontName=RLSmall-Regular
     outlineSettings {
      bApplyOutlineToDropShadows=false
      bMiteredCorners=false
      bSeparateFillAlpha=false
      outlineMaterial=null
      outlineSize=0
      outlineColor=<rgba 0 0 0 1>
     }
    }
   }
   slot {class=GridSlot}
  }
  TextBlock {
   widgetClassPath=&TextBlock
   parent=UIKitGridPanelAudioPreviewControls
   widget=TextBlock
   properties {
    text=Engine
    font {
     bForceMonospaced=false
     bMaterialIsStencil=false
     fontMaterial=null
     fontObject=null
     letterSpacing=0
     monospacedWidth=1
     size=9
     skewAmount=0
     typefaceFontName=RLSmall-Regular
     outlineSettings {
      bApplyOutlineToDropShadows=false
      bMiteredCorners=false
      bSeparateFillAlpha=false
      outlineMaterial=null
      outlineSize=0
      outlineColor=<rgba 0 0 0 1>
     }
    }
   }
   slot {class=GridSlot, properties {row=2}}
  }
  TextBlock_1 {
   widgetClassPath=&TextBlock
   parent=UIKitGridPanelAudioPreviewControls
   widget=TextBlock_1
   properties {
    text=Exhaust
    font {
     bForceMonospaced=false
     bMaterialIsStencil=false
     fontMaterial=null
     fontObject=null
     letterSpacing=0
     monospacedWidth=1
     size=9
     skewAmount=0
     typefaceFontName=RLSmall-Regular
     outlineSettings {
      bApplyOutlineToDropShadows=false
      bMiteredCorners=false
      bSeparateFillAlpha=false
      outlineMaterial=null
      outlineSize=0
      outlineColor=<rgba 0 0 0 1>
     }
    }
   }
   slot {class=GridSlot, properties {horizontalAlignment=HAlign_Right, row=2}}
  }
  RadialSlider_Azimuth {
   widgetClassPath=/Script/AdvancedWidgets.RadialSlider
   bIsVariable=true
   parent=UIKitGridPanelAudioPreviewControls
   widget=RadialSlider_Azimuth
   properties {
    showSliderHand=true
    value=-180
    sliderRange {
     externalCurve=null
     editorCurveData {
      defaultValue=3.4028234663852886e+38
      postInfinityExtrap=RCCE_Constant
      preInfinityExtrap=RCCE_Constant
      keys [
       {
        arriveTangent=0
        arriveTangentWeight=0
        interpMode=RCIM_Linear
        leaveTangent=0
        leaveTangentWeight=0
        tangentMode=RCTM_Auto
        tangentWeightMode=RCTWM_WeightedNone
        time=0
        value=-180
       }
       {
        arriveTangent=0
        arriveTangentWeight=0
        interpMode=RCIM_Linear
        leaveTangent=0
        leaveTangentWeight=0
        tangentMode=RCTM_Auto
        tangentWeightMode=RCTWM_WeightedNone
        time=1
        value=180
       }
      ]
     }
    }
   }
   slot {class=GridSlot, properties {row=1}}
  }
  RadialSlider_Distance {
   widgetClassPath=&RadialSlider
   bIsVariable=true
   parent=UIKitGridPanelAudioPreviewControls
   widget=RadialSlider_Distance
   properties {
    bUseCustomDefaultValue=true
    customDefaultValue=-100
    showSliderHand=true
    sliderRange {
     externalCurve=null
     editorCurveData {
      defaultValue=3.4028234663852886e+38
      postInfinityExtrap=RCCE_Constant
      preInfinityExtrap=RCCE_Constant
      keys [
       {
        arriveTangent=0
        arriveTangentWeight=0
        interpMode=RCIM_Linear
        leaveTangent=0
        leaveTangentWeight=0
        tangentMode=RCTM_Auto
        tangentWeightMode=RCTWM_WeightedNone
        time=0
        value=0
       }
       {
        arriveTangent=0
        arriveTangentWeight=0
        interpMode=RCIM_Linear
        leaveTangent=0
        leaveTangentWeight=0
        tangentMode=RCTM_Auto
        tangentWeightMode=RCTWM_WeightedNone
        time=1
        value=10000
       }
      ]
     }
    }
   }
   slot {class=GridSlot, properties {column=1, row=1}}
  }
  TextBlock_3 {
   widgetClassPath=&TextBlock
   parent=UIKitGridPanelAudioPreviewControls
   widget=TextBlock_3
   properties {
    text=Distance
    font {
     bForceMonospaced=false
     bMaterialIsStencil=false
     fontMaterial=null
     fontObject=null
     letterSpacing=0
     monospacedWidth=1
     size=24
     skewAmount=0
     typefaceFontName=RLSmall-Regular
     outlineSettings {
      bApplyOutlineToDropShadows=false
      bMiteredCorners=false
      bSeparateFillAlpha=false
      outlineMaterial=null
      outlineSize=0
      outlineColor=<rgba 0 0 0 1>
     }
    }
   }
   slot {class=GridSlot, properties {column=1, horizontalAlignment=HAlign_Center}}
  }
  TextBlock_ExampleAssetName_1 {
   widgetClassPath=&TextBlock
   bIsVariable=true
   parent=UIKitGridPanelAudioPreviewControls
   widget=TextBlock_ExampleAssetName_1
   properties {
    text=Distance Value
    font {
     bForceMonospaced=false
     bMaterialIsStencil=false
     fontMaterial=null
     fontObject=&Roboto
     letterSpacing=0
     monospacedWidth=1
     size=15
     skewAmount=0
     typefaceFontName=Bold
     outlineSettings {
      bApplyOutlineToDropShadows=false
      bMiteredCorners=false
      bSeparateFillAlpha=false
      outlineMaterial=null
      outlineSize=0
      outlineColor=<rgba 0 0 0 1>
     }
    }
   }
   slot {class=GridSlot, properties {column=1, horizontalAlignment=HAlign_Center, row=2, verticalAlignment=VAlign_Top}}
  }
  UIKitGridPanel_74 {
   widgetClassPath=&UIKitGridPanel
   parent=UIKitGridPanelAudioPreviewControls
   widget=UIKitGridPanel_74
   slot {class=GridSlot, properties {column=2, row=1}}
  }
  RadialSlider_Tachometer {
   widgetClassPath=&RadialSlider
   bIsVariable=true
   parent=UIKitGridPanel_74
   widget=RadialSlider_Tachometer
   slot {class=GridSlot, properties {column=2, row=1}}
  }
  AnalogSlider_Throttle {
   widgetClassPath=/Script/CommonUI.AnalogSlider
   bIsVariable=true
   parent=UIKitGridPanel_74
   widget=AnalogSlider_Throttle
   properties {
    renderTransform {angle=-90, scale=<1 1>, shear=<0 0>, translation=<0 0>}
    widgetStyle {
     barThickness=4
     disabledBarImage {
      bIsDynamicallyLoaded=false
      drawAs=RoundedBox
      imageType=NoImage
      mirroring=NoMirror
      resourceName=None
      resourceObject=null
      tiling=NoTile
      imageSize=<0 0>
      margin=<ltrb 0 0 0 0>
      outlineSettings {
       bUseBrushTransparency=false
       roundingType=FixedRadius
       width=1
       color {colorUseRule=UseColor_ColorTable, specifiedColor=<rgba 0 0 0 0>}
       cornerRadii=<2 2 2 2>
      }
      tintColor {colorUseRule=UseColor_Specified, specifiedColor=<rgba 0.0048 0.0048 0.0048 1>}
      uVRegion {bIsValid=false, max=<0 0>, min=<0 0>}
     }
     disabledThumbImage {
      bIsDynamicallyLoaded=false
      drawAs=RoundedBox
      imageType=NoImage
      mirroring=NoMirror
      resourceName=None
      resourceObject=null
      tiling=NoTile
      imageSize=<8 8>
      margin=<ltrb 0 0 0 0>
      outlineSettings {
       bUseBrushTransparency=false
       roundingType=HalfHeightRadius
       width=0
       color {colorUseRule=UseColor_Specified, specifiedColor=<rgba 0 0 0 0>}
       cornerRadii=<0 0 0 0>
      }
      tintColor {colorUseRule=UseColor_Specified, specifiedColor=<rgba 0.5271 0.5271 0.5271 1>}
      uVRegion {bIsValid=false, max=<0 0>, min=<0 0>}
     }
     hoveredBarImage {
      bIsDynamicallyLoaded=false
      drawAs=RoundedBox
      imageType=NoImage
      mirroring=NoMirror
      resourceName=None
      resourceObject=null
      tiling=NoTile
      imageSize=<0 0>
      margin=<ltrb 0 0 0 0>
      outlineSettings {
       bUseBrushTransparency=false
       roundingType=FixedRadius
       width=1
       color {colorUseRule=UseColor_ColorTable, specifiedColor=<rgba 0 0 0 0>}
       cornerRadii=<2 2 2 2>
      }
      tintColor {colorUseRule=UseColor_Specified, specifiedColor=<rgba 0.0048 0.0048 0.0048 1>}
      uVRegion {bIsValid=false, max=<0 0>, min=<0 0>}
     }
     hoveredThumbImage {
      bIsDynamicallyLoaded=false
      drawAs=RoundedBox
      imageType=NoImage
      mirroring=NoMirror
      resourceName=None
      resourceObject=null
      tiling=NoTile
      imageSize=<8 8>
      margin=<ltrb 0 0 0 0>
      outlineSettings {
       bUseBrushTransparency=false
       roundingType=HalfHeightRadius
       width=0
       color {colorUseRule=UseColor_Specified, specifiedColor=<rgba 0 0 0 0>}
       cornerRadii=<0 0 0 0>
      }
      tintColor {colorUseRule=UseColor_Specified, specifiedColor=<rgba 1 1 1 1>}
      uVRegion {bIsValid=false, max=<0 0>, min=<0 0>}
     }
     normalBarImage {
      bIsDynamicallyLoaded=false
      drawAs=RoundedBox
      imageType=NoImage
      mirroring=NoMirror
      resourceName=None
      resourceObject=null
      tiling=NoTile
      imageSize=<0 0>
      margin=<ltrb 0 0 0 0>
      outlineSettings {
       bUseBrushTransparency=false
       roundingType=FixedRadius
       width=1
       color {colorUseRule=UseColor_ColorTable, specifiedColor=<rgba 0 0 0 0>}
       cornerRadii=<2 2 2 2>
      }
      tintColor {colorUseRule=UseColor_Specified, specifiedColor=<rgba 0.0048 0.0048 0.0048 1>}
      uVRegion {bIsValid=false, max=<0 0>, min=<0 0>}
     }
     normalThumbImage {
      bIsDynamicallyLoaded=false
      drawAs=RoundedBox
      imageType=NoImage
      mirroring=NoMirror
      resourceName=None
      resourceObject=null
      tiling=NoTile
      imageSize=<8 8>
      margin=<ltrb 0 0 0 0>
      outlineSettings {
       bUseBrushTransparency=false
       roundingType=HalfHeightRadius
       width=0
       color {colorUseRule=UseColor_Specified, specifiedColor=<rgba 0 0 0 0>}
       cornerRadii=<0 0 0 0>
      }
      tintColor {colorUseRule=UseColor_Specified, specifiedColor=<rgba 0.5271 0.5271 0.5271 1>}
      uVRegion {bIsValid=false, max=<0 0>, min=<0 0>}
     }
    }
   }
   slot {class=GridSlot, properties {column=1, row=1}}
  }
  TextBlock_5 {
   widgetClassPath=&TextBlock
   parent=UIKitGridPanel_74
   widget=TextBlock_5
   properties {text=Throttle}
   slot {class=GridSlot, properties {column=1, horizontalAlignment=HAlign_Center, verticalAlignment=VAlign_Center}}
  }
  Text_GearNumber {
   widgetClassPath=&TextBlock
   bIsVariable=true
   parent=UIKitGridPanel_74
   widget=Text_GearNumber
   slot {class=GridSlot, properties {horizontalAlignment=HAlign_Left, verticalAlignment=VAlign_Top}}
   properties=~
  }
  HorizontalBoxAudioPlayback {
   widgetClassPath=&HorizontalBox
   parent=UIKitGridPanelAudioPreviewControls
   widget=HorizontalBoxAudioPlayback
   slot {class=GridSlot, properties {column=2, horizontalAlignment=HAlign_Center, row=3, verticalAlignment=VAlign_Center}}
  }
  AudioButtonToggle {
   widgetClassPath=/AudioWidgets/AudioButtonToggle/AudioButtonToggle
   bIsVariable=true
   parent=HorizontalBoxAudioPlayback
   widget=AudioButtonToggle
   properties {textValue=Preview Engine Metasound}
   slot {class=HorizontalBoxSlot}
  }
  AudioButtonToggle_FNVehicleMix {
   widgetClassPath=&AudioButtonToggle
   bIsVariable=true
   parent=HorizontalBoxAudioPlayback
   widget=AudioButtonToggle_FNVehicleMix
   properties {textValue=FN Vehicle Mix Toggle}
   slot {class=HorizontalBoxSlot}
  }
  UIKitGridPanelLoudness {
   widgetClassPath=&UIKitGridPanel
   parent=CommonActivatableWidgetSwitcher_171
   widget=UIKitGridPanelLoudness
   slot {class=WidgetSwitcherSlot}
  }
  HorizontalBoxDefaultButtons {
   widgetClassPath=&HorizontalBox
   parent=UIKitGridPanel_616
   widget=HorizontalBoxDefaultButtons
   slot {class=GridSlot, properties {row=1}}
  }
  UniformGridPanelCreationButtons {
   widgetClassPath=/Script/UMG.UniformGridPanel
   parent=HorizontalBoxDefaultButtons
   widget=UniformGridPanelCreationButtons
   properties {minDesiredSlotHeight=46.4643}
   slot {class=HorizontalBoxSlot, properties {padding=<ltrb 0 5 10 10>}}
  }
  EditorUtilityButton_56_MakeFolder {
   widgetClassPath=&EditorUtilityButton
   parent=UniformGridPanelCreationButtons
   widget=EditorUtilityButton_56_MakeFolder
   properties {toolTipText=Make metasounds from selected soundwaves}
   slot {class=UniformGridSlot, properties {horizontalAlignment=HAlign_Fill, verticalAlignment=VAlign_Fill}}
  }
  UEFN_TextBlock_C_4 {
   widgetClassPath=&UEFN_TextBlock
   parent=EditorUtilityButton_56_MakeFolder
   widget=UEFN_TextBlock_C_4
   properties {
    justification=Center
    text=1. Make Folder
    font {
     bForceMonospaced=false
     bMaterialIsStencil=false
     fontMaterial=null
     fontObject=&BurbankSmall
     letterSpacing=0
     monospacedWidth=1
     size=9
     skewAmount=0
     typefaceFontName=Black
     outlineSettings {
      bApplyOutlineToDropShadows=false
      bMiteredCorners=false
      bSeparateFillAlpha=false
      outlineMaterial=null
      outlineSize=0
      outlineColor=<rgba 0 0 0 1>
     }
    }
   }
   slot {class=ButtonSlot}
  }
  EditorUtilityButton_56_MakeBoost {
   widgetClassPath=&EditorUtilityButton
   parent=UniformGridPanelCreationButtons
   widget=EditorUtilityButton_56_MakeBoost
   properties {toolTipText=Make metasounds from selected soundwaves}
   slot {class=UniformGridSlot, properties {horizontalAlignment=HAlign_Fill, row=1, verticalAlignment=VAlign_Fill}}
  }
  UEFN_TextBlock_C_39 {
   widgetClassPath=&UEFN_TextBlock
   parent=EditorUtilityButton_56_MakeBoost
   widget=UEFN_TextBlock_C_39
   properties {
    justification=Center
    text=3. Make Engine
    font {
     bForceMonospaced=false
     bMaterialIsStencil=false
     fontMaterial=null
     fontObject=&BurbankSmall
     letterSpacing=0
     monospacedWidth=1
     size=9
     skewAmount=0
     typefaceFontName=Black
     outlineSettings {
      bApplyOutlineToDropShadows=false
      bMiteredCorners=false
      bSeparateFillAlpha=false
      outlineMaterial=null
      outlineSize=0
      outlineColor=<rgba 0 0 0 1>
     }
    }
   }
   slot {class=ButtonSlot}
  }
  EditorUtilityButton_56_MakeDataAssets {
   widgetClassPath=&EditorUtilityButton
   parent=UniformGridPanelCreationButtons
   widget=EditorUtilityButton_56_MakeDataAssets
   properties {toolTipText=Make metasounds from selected soundwaves}
   slot {class=UniformGridSlot, properties {column=1, horizontalAlignment=HAlign_Fill, row=1, verticalAlignment=VAlign_Fill}}
  }
  UEFN_TextBlock_C_5 {
   widgetClassPath=&UEFN_TextBlock
   parent=EditorUtilityButton_56_MakeDataAssets
   widget=UEFN_TextBlock_C_5
   properties {
    justification=Center
    text=4. Make Data Assets
    font {
     bForceMonospaced=false
     bMaterialIsStencil=false
     fontMaterial=null
     fontObject=&BurbankSmall
     letterSpacing=0
     monospacedWidth=1
     size=9
     skewAmount=0
     typefaceFontName=Black
     outlineSettings {
      bApplyOutlineToDropShadows=false
      bMiteredCorners=false
      bSeparateFillAlpha=false
      outlineMaterial=null
      outlineSize=0
      outlineColor=<rgba 0 0 0 1>
     }
    }
   }
   slot {class=ButtonSlot}
  }
  EditorUtilityButton_56_Importfiles {
   widgetClassPath=&EditorUtilityButton
   parent=UniformGridPanelCreationButtons
   widget=EditorUtilityButton_56_Importfiles
   properties {toolTipText=Make metasounds from selected soundwaves}
   slot {class=UniformGridSlot, properties {column=1, horizontalAlignment=HAlign_Fill, verticalAlignment=VAlign_Fill}}
  }
  UEFN_TextBlock_C_6 {
   widgetClassPath=&UEFN_TextBlock
   parent=EditorUtilityButton_56_Importfiles
   widget=UEFN_TextBlock_C_6
   properties {
    justification=Center
    text=2. Import Files
    font {
     bForceMonospaced=false
     bMaterialIsStencil=false
     fontMaterial=null
     fontObject=&BurbankSmall
     letterSpacing=0
     monospacedWidth=1
     size=9
     skewAmount=0
     typefaceFontName=Black
     outlineSettings {
      bApplyOutlineToDropShadows=false
      bMiteredCorners=false
      bSeparateFillAlpha=false
      outlineMaterial=null
      outlineSize=0
      outlineColor=<rgba 0 0 0 1>
     }
    }
   }
   slot {class=ButtonSlot}
  }
  UniformGridPanelNavigation {
   widgetClassPath=&UniformGridPanel
   parent=HorizontalBoxDefaultButtons
   widget=UniformGridPanelNavigation
   properties {minDesiredSlotHeight=43.3772, minDesiredSlotWidth=27.0677}
   slot {class=HorizontalBoxSlot, properties {horizontalAlignment=HAlign_Left, padding=<ltrb 0 5 0 10>}}
  }
  EditorUtilityButton_56_GoToVehicleAudioDataAssets {
   widgetClassPath=&EditorUtilityButton
   parent=UniformGridPanelNavigation
   widget=EditorUtilityButton_56_GoToVehicleAudioDataAssets
   properties {backgroundColor=<rgba 0.0012 0.0595 0.3231 1>}
   slot {class=UniformGridSlot, properties {horizontalAlignment=HAlign_Fill, verticalAlignment=VAlign_Fill}}
  }
  UEFN_TextBlock_C_8 {
   widgetClassPath=&UEFN_TextBlock
   parent=EditorUtilityButton_56_GoToVehicleAudioDataAssets
   widget=UEFN_TextBlock_C_8
   properties {
    justification=Center
    text=GoTo Vehicle Audio Data Assets
    font {
     bForceMonospaced=false
     bMaterialIsStencil=false
     fontMaterial=null
     fontObject=&BurbankSmall
     letterSpacing=0
     monospacedWidth=1
     size=9
     skewAmount=0
     typefaceFontName=Black
     outlineSettings {
      bApplyOutlineToDropShadows=false
      bMiteredCorners=false
      bSeparateFillAlpha=false
      outlineMaterial=null
      outlineSize=0
      outlineColor=<rgba 0 0 0 1>
     }
    }
   }
   slot {class=ButtonSlot}
  }
  EditorUtilityButton_56_GoToVehicleCosmetic {
   widgetClassPath=&EditorUtilityButton
   parent=UniformGridPanelNavigation
   widget=EditorUtilityButton_56_GoToVehicleCosmetic
   properties {backgroundColor=<rgba 0.0012 0.0595 0.3231 1>}
   slot {class=UniformGridSlot, properties {horizontalAlignment=HAlign_Fill, row=1, verticalAlignment=VAlign_Fill}}
  }
  UEFN_TextBlock_C_3 {
   widgetClassPath=&UEFN_TextBlock
   parent=EditorUtilityButton_56_GoToVehicleCosmetic
   widget=UEFN_TextBlock_C_3
   properties {
    justification=Center
    text=GoTo Vehicle Bodies
    font {
     bForceMonospaced=false
     bMaterialIsStencil=false
     fontMaterial=null
     fontObject=&BurbankSmall
     letterSpacing=0
     monospacedWidth=1
     size=9
     skewAmount=0
     typefaceFontName=Black
     outlineSettings {
      bApplyOutlineToDropShadows=false
      bMiteredCorners=false
      bSeparateFillAlpha=false
      outlineMaterial=null
      outlineSize=0
      outlineColor=<rgba 0 0 0 1>
     }
    }
   }
   slot {class=ButtonSlot}
  }
  EditorUtilityButton_56_GoToVehicleAudio {
   widgetClassPath=&EditorUtilityButton
   parent=UniformGridPanelNavigation
   widget=EditorUtilityButton_56_GoToVehicleAudio
   properties {toolTipText=Make metasounds from selected soundwaves, backgroundColor=<rgba 0.0012 0.0595 0.3231 1>}
   slot {class=UniformGridSlot, properties {column=1, horizontalAlignment=HAlign_Fill, verticalAlignment=VAlign_Fill}}
  }
  UEFN_TextBlock_C_7 {
   widgetClassPath=&UEFN_TextBlock
   parent=EditorUtilityButton_56_GoToVehicleAudio
   widget=UEFN_TextBlock_C_7
   properties {
    justification=Center
    text=GoTo VehicleAudio/Engine
    font {
     bForceMonospaced=false
     bMaterialIsStencil=false
     fontMaterial=null
     fontObject=&BurbankSmall
     letterSpacing=0
     monospacedWidth=1
     size=9
     skewAmount=0
     typefaceFontName=Black
     outlineSettings {
      bApplyOutlineToDropShadows=false
      bMiteredCorners=false
      bSeparateFillAlpha=false
      outlineMaterial=null
      outlineSize=0
      outlineColor=<rgba 0 0 0 1>
     }
    }
   }
   slot {class=ButtonSlot}
  }
 ]
}
